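# Control Rig for the medium male Wombat Poncho body that adjusts sleeve and armpit bones as the arms move.
# Spherical pose readers on each lower arm output weights for ModifyTransforms nodes that offset and rotate those dyn_ bones.
# A Sequence runs front correctives for both arms, then back correctives.
# Only back weights are smoothed through AlphaInterp, and their readers have debug drawing enabled.
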ControlRigBlueprint /BRCosmetics/Characters/Player/Male/Medium/Bodies/M_MED_Wombat_Poncho/Meshes/M_MED_Wombat_Poncho_CtrlRig
graphs [
 RigVMFunctionLibrary {links [], nodes []}
 RigVMModel {nodes [
   RigUnit_BeginExecution = &RigUnit_BeginExecution()
   RigVMFunction_Sequence = &RigVMFunction_Sequence(ExecuteContext=@RigUnit_BeginExecution.ExecutePin)
   n1 = &RigUnit_SphericalPoseReader(
    ActiveRegionScaleFactors=(PositiveWidth=1,NegativeWidth=1,PositiveHeight=1,NegativeHeight=1)
    ActiveRegionSize=0.1
    Debug=(bDrawDebug=false,DebugScale=25,DebugSegments=20,DebugThickness=0.25)
    DriverAxis=(X=1,Y=0,Z=0)
    DriverItem=(Type=Bone,Name="lowerarm_l")
    FalloffRegionScaleFactors=(PositiveWidth=1,NegativeWidth=1,PositiveHeight=1,NegativeHeight=1)
    FalloffSize=0.4
    FlipHeightScaling=false
    FlipWidthScaling=false
    OptionalParentItem=(Type=Bone)
    RotationOffset=(X=90,Y=0,Z=0)
    ExecutePin=@RigVMFunction_Sequence.A
   )
   > &RigUnit_ModifyTransforms(
    ItemToModify="((Item=(Type=Bone,Name=\"dyn_l_sleeve_front_master\"),Transform=(Rotation=(X=0,Y=0,Z=0,W=1),Translation=(X=-3,Y=-2,Z=0),Scale3D=(X=1,Y=1,Z=1))),(Item=(Type=Bone,Name=\"dyn_poncho_armpit_front_l\"),Transform=(Rotation=(X=0,Y=0,Z=0,W=1),Translation=(X=6,Y=-5,Z=0 ...(truncated)"
    ItemToModify.1.Transform.Rotation=&RigVMFunction_MathQuaternionFromEuler(Euler=(X=0,Y=0,Z=-5)).Result
    ItemToModify.0.Transform.Rotation=&RigVMFunction_MathQuaternionFromEuler(Euler=(X=0,Y=10,Z=0)).Result
    Weight=@n1.OutputParam
   )
   > n2 = &RigUnit_SphericalPoseReader(
    ActiveRegionScaleFactors=(PositiveWidth=1,NegativeWidth=1,PositiveHeight=1,NegativeHeight=1)
    ActiveRegionSize=0.1
    Debug=(bDrawDebug=false,DebugScale=25,DebugSegments=20,DebugThickness=0.25)
    DriverAxis=(X=-1,Y=0,Z=0)
    DriverItem=(Type=Bone,Name="lowerarm_r")
    FalloffRegionScaleFactors=(PositiveWidth=1,NegativeWidth=1,PositiveHeight=1,NegativeHeight=1)
    FalloffSize=0.4
    FlipHeightScaling=false
    FlipWidthScaling=false
    OptionalParentItem=(Type=Bone)
    RotationOffset=(X=-90,Y=0,Z=0)
   )
   > &RigUnit_ModifyTransforms(
    ItemToModify="((Item=(Type=Bone,Name=\"dyn_r_sleeve_front_master\"),Transform=(Rotation=(X=0,Y=0,Z=0,W=1),Translation=(X=2,Y=2,Z=-2),Scale3D=(X=1,Y=1,Z=1))),(Item=(Type=Bone,Name=\"dyn_poncho_armpit_front_r\"),Transform=(Rotation=(X=0,Y=0,Z=0,W=1),Translation=(X=0,Y=0,Z=0), ...(truncated)"
    ItemToModify.0.Transform.Rotation=&RigVMFunction_MathQuaternionFromEuler(Euler=(X=10,Y=-15,Z=0)).Result
    Weight=@n2.OutputParam
   )
   n3 = &RigUnit_SphericalPoseReader(
    ActiveRegionScaleFactors=(PositiveWidth=1,NegativeWidth=1,PositiveHeight=1,NegativeHeight=1)
    ActiveRegionSize=0.1
    Debug=(bDrawDebug=true,DebugScale=25,DebugSegments=20,DebugThickness=0.25)
    DriverAxis=(X=1,Y=0,Z=0)
    DriverItem=(Type=Bone,Name="lowerarm_l")
    FalloffRegionScaleFactors=(PositiveWidth=0.5,NegativeWidth=0.5,PositiveHeight=0.5,NegativeHeight=0.5)
    FalloffSize=0.4
    FlipHeightScaling=false
    FlipWidthScaling=false
    OptionalParentItem=(Type=Bone)
    RotationOffset=(X=-20,Y=-90,Z=0)
    ExecutePin=@RigVMFunction_Sequence.B
   )
   AlphaInterp = &RigVMFunction_AlphaInterp(InterpSpeedDecreasing=10, InterpSpeedIncreasing=10, Scale=1, bClampResult=False, Value=@n3.OutputParam)
   > &RigUnit_ModifyTransforms(
    ItemToModify="((Item=(Type=Bone,Name=\"dyn_l_sleeve_back_01\"),Transform=(Rotation=(X=0,Y=0,Z=0,W=1),Translation=(X=0,Y=0,Z=0),Scale3D=(X=1,Y=1,Z=1))),(Item=(Type=Bone,Name=\"dyn_poncho_armpit_back_l\"),Transform=(Rotation=(X=0,Y=0,Z=0,W=1),Translation=(X=-5,Y=5,Z=0),Scale3 ...(truncated)"
    Weight=@AlphaInterp.Result
    ItemToModify.0.Transform.Rotation=&RigVMFunction_MathQuaternionFromEuler(Euler=(X=0,Y=-20,Z=0)).Result
    ItemToModify.1.Transform.Rotation=&RigVMFunction_MathQuaternionFromEuler(Euler=(X=0,Y=0,Z=0)).Result
   )
   > n4 = &RigUnit_SphericalPoseReader(
    ActiveRegionScaleFactors=(PositiveWidth=1,NegativeWidth=1,PositiveHeight=1,NegativeHeight=1)
    ActiveRegionSize=0.1
    Debug=(bDrawDebug=true,DebugScale=25,DebugSegments=20,DebugThickness=0.25)
    DriverAxis=(X=-1,Y=0,Z=0)
    DriverItem=(Type=Bone,Name="lowerarm_r")
    FalloffRegionScaleFactors=(PositiveWidth=0.5,NegativeWidth=0.5,PositiveHeight=0.5,NegativeHeight=0.5)
    FalloffSize=0.4
    FlipHeightScaling=false
    FlipWidthScaling=false
    OptionalParentItem=(Type=Bone)
    RotationOffset=(X=15,Y=90,Z=0)
   )
   n5 = &RigVMFunction_AlphaInterp(InterpSpeedDecreasing=10, InterpSpeedIncreasing=10, Scale=1, bClampResult=False, Value=@n4.OutputParam)
   > &RigUnit_ModifyTransforms(
    ItemToModify="((Item=(Type=Bone,Name=\"dyn_r_sleeve_back_01\"),Transform=(Rotation=(X=0,Y=0,Z=0,W=1),Translation=(X=0,Y=0,Z=0),Scale3D=(X=1,Y=1,Z=1))),(Item=(Type=Bone,Name=\"dyn_poncho_armpit_back_r\"),Transform=(Rotation=(X=0,Y=0,Z=0,W=1),Translation=(X=2,Y=-5,Z=0),Scale3 ...(truncated)"
    Weight=@n5.Result
    ItemToModify.0.Transform.Rotation=&RigVMFunction_MathQuaternionFromEuler(Euler=(X=0,Y=-10,Z=0)).Result
    ItemToModify.1.Transform.Rotation=&RigVMFunction_MathQuaternionFromEuler(Euler=(X=0,Y=0,Z=0)).Result
    ItemToModify.2.Transform.Rotation=&RigVMFunction_MathQuaternionFromEuler(Euler=(X=0,Y=-20,Z=0)).Result
    ItemToModify.3.Transform.Rotation=&RigVMFunction_MathQuaternionFromEuler(Euler=(X=0,Y=-20,Z=0)).Result
   )
  ]}
]
hierarchy [
 root {loc=<0 0 0>, rot=<0 0 0 1>}
 attach {parent=root, loc=<0 0 0>, rot=<0 0 0 1>}
 pelvis {parent=root, loc=<0 -0.7071 93.1191>, rot=<0.0127 0.707 -0.0127 -0.707>}
 spine_01 {parent=pelvis, loc=<6.7494 -0.6388 0>, rot=<0 0 0.034 -0.9994>}
 spine_02 {parent=spine_01, loc=<7.2518 -0.035 0>, rot=<0 0 0.0238 -0.9997>}
 spine_03 {parent=spine_02, loc=<6.0104 0 0>, rot=<0 0 0.0149 -0.9999>}
 spine_04 {parent=spine_03, loc=<8.1024 0 0>, rot=<0 0 -0.0132 -0.9999>}
 spine_05 {parent=spine_04, loc=<10.0146 0 0>, rot=<0 0 -0.0015 -1>}
 clavicle_l {parent=spine_05, loc=<4.3575 -0.1466 -2.3223>, rot=<0.0465 -0.7289 -0.009 -0.6829>}
 upperarm_l {parent=clavicle_l, loc=<15.5111 0 0>, rot=<0.0226 -0.3234 -0.0088 -0.9459>}
 lowerarm_l {parent=upperarm_l, loc=<26.6514 0 0>, rot=<0.0391 -0.0099 -0.2453 -0.9686>}
 hand_l {parent=lowerarm_l, loc=<22.3491 0 0>, rot=<-0.7081 -0.0507 0.0975 -0.6975>}
 index_metacarpal_l {parent=hand_l, loc=<1.6535 0.0183 -2.96>, rot=<0.0158 -0.0124 -0.0173 -0.9996>}
 index_01_l {parent=index_metacarpal_l, loc=<8.3008 0.0357 -0.0524>, rot=<-0.0044 0.0485 0.0924 -0.9945>}
 index_02_l {parent=index_01_l, loc=<3.3587 -0.3004 0.0452>, rot=<-0.002 0.0027 0.3225 -0.9466>}
 index_03_l {parent=index_02_l, loc=<2.5578 -0.123 -0.0252>, rot=<-0.002 0.0067 0.2558 -0.9667>}
 middle_metacarpal_l {parent=hand_l, loc=<2.1856 -0.1725 -0.5984>, rot=<0.006 0.0168 -0.0329 -0.9993>}
 middle_01_l {parent=middle_metacarpal_l, loc=<7.3275 -0.2447 0.0129>, rot=<-0.0208 0.0042 0.0858 -0.9961>}
 middle_02_l {parent=middle_01_l, loc=<4.0123 0.1401 0.1175>, rot=<-0.0218 0.0258 0.3512 -0.9357>}
 middle_03_l {parent=middle_02_l, loc=<2.6077 0.0534 0.1373>, rot=<0.0082 0.0188 0.1708 -0.9851>}
 pinky_metacarpal_l {parent=hand_l, loc=<1.7616 -0.9753 3.3814>, rot=<-0.0397 0.0814 -0.069 -0.9935>}
 pinky_01_l {parent=pinky_metacarpal_l, loc=<7.1709 -0.4831 -0.1094>, rot=<-0.0654 -0.0035 0.2242 -0.9723>}
 pinky_02_l {parent=pinky_01_l, loc=<2.8274 0.0893 -0.1043>, rot=<0.0037 0.0011 0.2854 -0.9584>}
 pinky_03_l {parent=pinky_02_l, loc=<2.2046 0.0307 -0.0078>, rot=<-0.0002 0.0098 0.1445 -0.9895>}
 ring_metacarpal_l {parent=hand_l, loc=<2.2074 -0.3625 1.3638>, rot=<-0.0284 0.0722 -0.0458 -0.9959>}
 ring_01_l {parent=ring_metacarpal_l, loc=<7.0295 -0.6197 -0.1446>, rot=<-0.0489 -0.0252 0.1892 -0.9804>}
 ring_02_l {parent=ring_01_l, loc=<3.6174 0.4647 -0.1027>, rot=<0 0 0.2806 -0.9598>}
 ring_03_l {parent=ring_02_l, loc=<2.4152 0.114 0.0199>, rot=<0 0 0.1189 -0.9929>}
 thumb_01_l {parent=hand_l, loc=<1.3118 -0.3191 -2.459>, rot=<0.6546 -0.4042 0.0664 -0.6354>}
 thumb_02_l {parent=thumb_01_l, loc=<5.9336 -0.0448 0.0401>, rot=<-0.0037 0.0213 0.298 -0.9543>}
 thumb_03_l {parent=thumb_02_l, loc=<2.2223 0.3682 -0.0907>, rot=<0.0118 -0.013 0.0066 -0.9998>}
 weapon_l {parent=hand_l, loc=<-0.1974 -3.4111 0.3811>, rot=<0 0 0 1>}
 lowerarm_twist_01_l {parent=lowerarm_l, loc=<17.8549 0 0>, rot=<0 0 0 1>}
 lowerarm_twist_02_l {parent=lowerarm_l, loc=<9.6199 0 0>, rot=<0 0 0 1>}
 upperarm_twist_01_l {parent=upperarm_l, loc=<0.0041 0 0>, rot=<0 0 0 1>}
 upperarm_twist_02_l {parent=upperarm_l, loc=<13.7941 0 0>, rot=<0 0 0 1>}
 dyn_l_sleeve_mid_master {parent=upperarm_l, loc=<12.7624 -1.088 5.4914>, rot=<-0.0091 0.0912 -0.017 -0.9956>}
 dyn_l_sleeve_mid_01 {parent=dyn_l_sleeve_mid_master, loc=<0 0 0>, rot=<0 0 0 1>}
 dyn_l_sleeve_mid_02 {parent=dyn_l_sleeve_mid_01, loc=<12.1143 0 0>, rot=<0 0.0794 0 -0.9968>}
 dyn_l_sleeve_back_master {parent=upperarm_l, loc=<13.7884 -6.7902 1.7778>, rot=<-0.4767 0.0724 -0.058 -0.8741>}
 dyn_l_sleeve_back_01 {parent=dyn_l_sleeve_back_master, loc=<0 0 0>, rot=<0 0 0 1>}
 dyn_l_sleeve_back_02 {parent=dyn_l_sleeve_back_01, loc=<12.1143 0 0>, rot=<0 0.0794 0 -0.9968>}
 dyn_l_sleeve_front_master {parent=upperarm_l, loc=<12.9871 5.5097 3.5022>, rot=<0.2781 0.0922 0.01 -0.9561>}
 dyn_l_sleeve_front_01 {parent=dyn_l_sleeve_front_master, loc=<0 0 0>, rot=<0 0 0 1>}
 dyn_l_sleeve_front_02 {parent=dyn_l_sleeve_front_01, loc=<12.1143 0 0>, rot=<0 0.0794 0 -0.9968>}
 dyn_poncho_armpit_back_l {parent=upperarm_l, loc=<21.8409 -6.2256 -6.5634>, rot=<0.0924 -0.0058 0.9955 -0.0184>}
 dyn_poncho_armpit_front_l {parent=upperarm_l, loc=<21.2635 9.9852 -6.3195>, rot=<0.0093 -0.9902 -0.1035 0.0938>}
 clavicle_r {parent=spine_05, loc=<4.3575 -0.1466 2.3223>, rot=<0.7289 0.0465 -0.6829 0.009>}
 upperarm_r {parent=clavicle_r, loc=<-15.5111 0 0>, rot=<0.0226 -0.3234 -0.0088 -0.9459>}
 lowerarm_r {parent=upperarm_r, loc=<-26.6515 0 0>, rot=<0.0391 -0.0099 -0.2453 -0.9686>}
 hand_r {parent=lowerarm_r, loc=<-22.3492 0 0>, rot=<-0.7081 -0.0507 0.0975 -0.6975>}
 index_metacarpal_r {parent=hand_r, loc=<-1.6534 -0.0181 2.9601>, rot=<0.0158 -0.0124 -0.0173 -0.9996>}
 index_01_r {parent=index_metacarpal_r, loc=<-8.3008 -0.0357 0.0524>, rot=<-0.0044 0.0485 0.0924 -0.9945>}
 index_02_r {parent=index_01_r, loc=<-3.3589 0.3003 -0.0453>, rot=<-0.002 0.0027 0.3225 -0.9466>}
 index_03_r {parent=index_02_r, loc=<-2.5578 0.1231 0.0251>, rot=<-0.002 0.0067 0.2558 -0.9667>}
 middle_metacarpal_r {parent=hand_r, loc=<-2.1856 0.1726 0.5984>, rot=<0.006 0.0168 -0.0329 -0.9993>}
 middle_01_r {parent=middle_metacarpal_r, loc=<-7.3273 0.2447 -0.0129>, rot=<-0.0208 0.0042 0.0858 -0.9961>}
 middle_02_r {parent=middle_01_r, loc=<-4.0142 -0.1411 -0.07>, rot=<-0.0217 0.0259 0.3512 -0.9357>}
 middle_03_r {parent=middle_02_r, loc=<-2.6074 -0.0534 -0.1312>, rot=<0.0083 0.0188 0.1708 -0.9851>}
 pinky_metacarpal_r {parent=hand_r, loc=<-1.7617 0.9753 -3.3814>, rot=<-0.0397 0.0814 -0.069 -0.9935>}
 pinky_01_r {parent=pinky_metacarpal_r, loc=<-7.1707 0.4832 0.1094>, rot=<-0.0654 -0.0035 0.2242 -0.9723>}
 pinky_02_r {parent=pinky_01_r, loc=<-2.8274 -0.0892 0.1043>, rot=<0.0037 0.0011 0.2854 -0.9584>}
 pinky_03_r {parent=pinky_02_r, loc=<-2.2045 -0.031 0.0078>, rot=<-0.0002 0.0098 0.1445 -0.9895>}
 ring_metacarpal_r {parent=hand_r, loc=<-2.2073 0.3627 -1.3638>, rot=<-0.0284 0.0722 -0.0458 -0.9959>}
 ring_01_r {parent=ring_metacarpal_r, loc=<-7.0297 0.6197 0.1446>, rot=<-0.0489 -0.0252 0.1892 -0.9804>}
 ring_02_r {parent=ring_01_r, loc=<-3.6171 -0.4648 0.1027>, rot=<0 0 0.2806 -0.9598>}
 ring_03_r {parent=ring_02_r, loc=<-2.4153 -0.1141 -0.0199>, rot=<0 0 0.1189 -0.9929>}
 thumb_01_r {parent=hand_r, loc=<-1.3119 0.3192 2.459>, rot=<0.6546 -0.4042 0.0664 -0.6354>}
 thumb_02_r {parent=thumb_01_r, loc=<-5.9335 0.0447 -0.0401>, rot=<-0.0037 0.0213 0.298 -0.9543>}
 thumb_03_r {parent=thumb_02_r, loc=<-2.2223 -0.3682 0.0907>, rot=<0.0118 -0.013 0.0066 -0.9998>}
 weapon_r {parent=hand_r, loc=<0.1975 3.4112 -0.3811>, rot=<0 0 0 -1>}
 lowerarm_twist_01_r {parent=lowerarm_r, loc=<-17.855 0 0>, rot=<0 0 0 1>}
 lowerarm_twist_02_r {parent=lowerarm_r, loc=<-9.6199 0 0>, rot=<0 0 0 1>}
 upperarm_twist_01_r {parent=upperarm_r, loc=<-0.0041 0 0>, rot=<0 0 0 1>}
 upperarm_twist_02_r {parent=upperarm_r, loc=<-13.7942 0 0>, rot=<0 0 0 1>}
 dyn_r_sleeve_mid_master {parent=upperarm_r, loc=<-12.7624 1.088 -5.4914>, rot=<-0.0091 0.0912 -0.017 -0.9956>}
 dyn_r_sleeve_mid_01 {parent=dyn_r_sleeve_mid_master, loc=<0 0 0>, rot=<0 -1 0 0>}
 dyn_r_sleeve_mid_02 {parent=dyn_r_sleeve_mid_01, loc=<12.1143 0 0>, rot=<0 0.0795 0 -0.9968>}
 dyn_r_sleeve_back_master {parent=upperarm_r, loc=<-13.7882 6.7902 -1.7779>, rot=<-0.4767 0.0724 -0.058 -0.8741>}
 dyn_r_sleeve_back_01 {parent=dyn_r_sleeve_back_master, loc=<0 0 0>, rot=<0 -0.9847 0.1743 0>}
 dyn_r_sleeve_back_02 {parent=dyn_r_sleeve_back_01, loc=<12.1144 0.0001 -0.0001>, rot=<0.0852 -0.0065 0.0081 -0.9963>}
 dyn_r_sleeve_front_master {parent=upperarm_r, loc=<-12.987 -5.5097 -3.5023>, rot=<0.2781 0.0922 0.01 -0.9561>}
 dyn_r_sleeve_front_01 {parent=dyn_r_sleeve_front_master, loc=<0 0 0>, rot=<0 -1 0.0075 0>}
 dyn_r_sleeve_front_02 {parent=dyn_r_sleeve_front_01, loc=<12.1143 0 0>, rot=<-0.0296 0.0797 0.0494 -0.9952>}
 dyn_poncho_armpit_front_r {parent=upperarm_r, loc=<-21.2634 -9.9851 6.3192>, rot=<0.0093 -0.9902 -0.1035 0.0938>}
 dyn_poncho_armpit_back_r {parent=upperarm_r, loc=<-21.8407 6.2256 6.5631>, rot=<0.0924 -0.0058 0.9955 -0.0184>}
 neck_01 {parent=spine_05, loc=<11.3138 0.1529 0>, rot=<0 0 -0.2003 -0.9797>}
 neck_02 {parent=neck_01, loc=<4.5627 0 0>, rot=<0 0 0.0504 -0.9987>}
 head {parent=neck_02, loc=<5.3914 0 0>, rot=<0 0 0.115 -0.9934>}
 dyn_poncho_front_mr_master {parent=spine_04, loc=<4.8077 11.9946 9.1516>, rot=<0.0294 -0.7065 -0.7065 0.0294>}
 dyn_poncho_front_mr_01 {parent=dyn_poncho_front_mr_master, loc=<0 0 0>, rot=<0 0 0 1>}
 dyn_poncho_front_mr_02 {parent=dyn_poncho_front_mr_01, loc=<13.5419 0 1.5907>, rot=<0 0 0 1>}
 dyn_poncho_front_mr_03 {parent=dyn_poncho_front_mr_02, loc=<16.0331 0 0.9205>, rot=<0 0 0 1>}
 dyn_poncho_front_mr_04 {parent=dyn_poncho_front_mr_03, loc=<16.0331 0 0.7122>, rot=<0 0 0 1>}
 dyn_poncho_front_mr_05 {parent=dyn_poncho_front_mr_04, loc=<17.6301 0 0>, rot=<-0.5 0.5 0.5 -0.5>}
 dyn_poncho_front_ml_master {parent=spine_04, loc=<4.9607 12.54 -11.8191>, rot=<0.0294 -0.7065 -0.7065 0.0294>}
 dyn_poncho_front_ml_01 {parent=dyn_poncho_front_ml_master, loc=<0 0 0>, rot=<0 0 0 1>}
 dyn_poncho_front_ml_02 {parent=dyn_poncho_front_ml_01, loc=<13.5419 0 0.7943>, rot=<0 0 0 1>}
 dyn_poncho_front_ml_03 {parent=dyn_poncho_front_ml_02, loc=<16.0331 0 0.7693>, rot=<0 0 0 1>}
 dyn_poncho_front_ml_04 {parent=dyn_poncho_front_ml_03, loc=<16.0331 0 0.4105>, rot=<0 0 -0.0181 -0.9998>}
 dyn_poncho_front_ml_05 {parent=dyn_poncho_front_ml_04, loc=<11.3093 0 0>, rot=<-0.7071 0 0 -0.7071>}
 dyn_poncho_front_mid_master {parent=spine_04, loc=<4.7115 14.186 0>, rot=<0.0294 -0.7065 -0.7065 0.0294>}
 dyn_poncho_front_mid_01 {parent=dyn_poncho_front_mid_master, loc=<0 0 0>, rot=<0 0 0 1>}
 dyn_poncho_front_mid_02 {parent=dyn_poncho_front_mid_01, loc=<13.5419 0 0.5976>, rot=<0 0 0 1>}
 dyn_poncho_front_mid_03 {parent=dyn_poncho_front_mid_02, loc=<16.0331 0 0.7561>, rot=<0 0 0 1>}
 dyn_poncho_front_mid_04 {parent=dyn_poncho_front_mid_03, loc=<16.0331 0 0>, rot=<0 0 -0.0181 -0.9998>}
 dyn_poncho_front_mid_05 {parent=dyn_poncho_front_mid_04, loc=<13.4565 0 0>, rot=<-0.7071 0 0 -0.7071>}
 dyn_poncho_back_mid_master {parent=spine_04, loc=<6.9961 -13.2096 0.4318>, rot=<0.7065 0.0294 -0.0294 -0.7065>}
 dyn_poncho_back_mid_01 {parent=dyn_poncho_back_mid_master, loc=<0 0 0>, rot=<0 -1 0 0>}
 dyn_poncho_back_mid_02 {parent=dyn_poncho_back_mid_01, loc=<13.542 0 1.3982>, rot=<0 0 0 1>}
 dyn_poncho_back_mid_03 {parent=dyn_poncho_back_mid_02, loc=<16.0327 0 0>, rot=<0 0 0 1>}
 dyn_poncho_back_mid_04 {parent=dyn_poncho_back_mid_03, loc=<16.0332 0 0>, rot=<0 0 0.0181 -0.9998>}
 dyn_poncho_back_mid_05 {parent=dyn_poncho_back_mid_04, loc=<13.4565 0 0>, rot=<0 -0.7071 0.7071 0>}
 dyn_poncho_back_ml_master {parent=spine_04, loc=<6.9696 -11.5449 -11.8191>, rot=<0.7065 0.0294 -0.0294 -0.7065>}
 dyn_poncho_back_ml_01 {parent=dyn_poncho_back_ml_master, loc=<0 0 0>, rot=<0 -1 0 0>}
 dyn_poncho_back_ml_02 {parent=dyn_poncho_back_ml_01, loc=<13.542 0 1.3982>, rot=<0 0 0 1>}
 dyn_poncho_back_ml_03 {parent=dyn_poncho_back_ml_02, loc=<16.0332 0 0.9278>, rot=<0 0 0 1>}
 dyn_poncho_back_ml_04 {parent=dyn_poncho_back_ml_03, loc=<16.0332 0 0>, rot=<0 0 0.0181 -0.9998>}
 dyn_poncho_back_ml_05 {parent=dyn_poncho_back_ml_04, loc=<11.3093 0 0>, rot=<0 -0.7071 0.7071 0>}
 dyn_poncho_back_mr_master {parent=spine_04, loc=<6.7282 -11.0325 11.1482>, rot=<0.7065 0.0294 -0.0294 -0.7065>}
 dyn_poncho_back_mr_01 {parent=dyn_poncho_back_mr_master, loc=<0 0 0>, rot=<0 -1 0 0>}
 dyn_poncho_back_mr_02 {parent=dyn_poncho_back_mr_01, loc=<13.542 0 1.7065>, rot=<0 0 0 1>}
 dyn_poncho_back_mr_03 {parent=dyn_poncho_back_mr_02, loc=<16.0331 0 0.6195>, rot=<0 0 0 1>}
 dyn_poncho_back_mr_04 {parent=dyn_poncho_back_mr_03, loc=<16.0331 0 0>, rot=<0 0 0 1>}
 dyn_poncho_back_mr_05 {parent=dyn_poncho_back_mr_04, loc=<13.303 0 0>, rot=<0.5 -0.5 0.5 -0.5>}
 dyn_poncho_front_r_master {parent=spine_04, loc=<5.027 8.303 21.5881>, rot=<-0.0294 0.7065 0.7065 -0.0294>}
 dyn_poncho_front_r_01 {parent=dyn_poncho_front_r_master, loc=<0 0 0>, rot=<0 0 0 1>}
 dyn_poncho_front_r_02 {parent=dyn_poncho_front_r_01, loc=<13.5419 0.6194 2.5233>, rot=<0 0 0 1>}
 dyn_poncho_front_r_03 {parent=dyn_poncho_front_r_02, loc=<16.0331 0 1.0501>, rot=<0.0087 0 0 -1>}
 dyn_poncho_front_r_04 {parent=dyn_poncho_front_r_03, loc=<16.0331 0 0>, rot=<0 0 0 1>}
 dyn_poncho_front_r_05 {parent=dyn_poncho_front_r_04, loc=<17.6301 0 0>, rot=<-0.5 0.5 0.5 -0.5>}
 dyn_poncho_front_l_master {parent=spine_04, loc=<5.3795 8.4255 -19.8135>, rot=<-0.0294 0.7065 0.7065 -0.0294>}
 dyn_poncho_front_l_01 {parent=dyn_poncho_front_l_master, loc=<0 0 0>, rot=<0 0 0 1>}
 dyn_poncho_front_l_02 {parent=dyn_poncho_front_l_01, loc=<13.5419 0 1.921>, rot=<0 0 0 1>}
 dyn_poncho_front_l_03 {parent=dyn_poncho_front_l_02, loc=<16.0331 0 1.2636>, rot=<0 0 0 1>}
 dyn_poncho_front_l_04 {parent=dyn_poncho_front_l_03, loc=<16.0331 0 0>, rot=<0 0 -0.0181 -0.9998>}
 dyn_poncho_front_l_05 {parent=dyn_poncho_front_l_04, loc=<11.5157 0 0>, rot=<-0.7071 0 0 -0.7071>}
 dyn_poncho_back_r_master {parent=spine_04, loc=<6.3332 -7.3555 23.1609>, rot=<0.7065 0.0294 -0.0294 -0.7065>}
 dyn_poncho_back_r_01 {parent=dyn_poncho_back_r_master, loc=<0 0 0>, rot=<0 -1 0 0>}
 dyn_poncho_back_r_02 {parent=dyn_poncho_back_r_01, loc=<13.542 0 2.688>, rot=<0 0 0 1>}
 dyn_poncho_back_r_03 {parent=dyn_poncho_back_r_02, loc=<16.0333 0 2.6339>, rot=<0 0 0 1>}
 dyn_poncho_back_r_04 {parent=dyn_poncho_back_r_03, loc=<16.0331 0 0.6047>, rot=<0 0 0 1>}
 dyn_poncho_back_r_05 {parent=dyn_poncho_back_r_04, loc=<12.6001 0 0>, rot=<0.5 -0.5 0.5 -0.5>}
 dyn_poncho_back_l_master {parent=spine_04, loc=<6.7007 -7.418 -22.8035>, rot=<0.7065 0.0294 -0.0294 -0.7065>}
 dyn_poncho_back_l_01 {parent=dyn_poncho_back_l_master, loc=<0 0 0>, rot=<0 -1 0 0>}
 dyn_poncho_back_l_02 {parent=dyn_poncho_back_l_01, loc=<13.542 0 3.6395>, rot=<0 0 0 1>}
 dyn_poncho_back_l_03 {parent=dyn_poncho_back_l_02, loc=<16.0328 0 1.6824>, rot=<0 0 0 1>}
 dyn_poncho_back_l_04 {parent=dyn_poncho_back_l_03, loc=<16.0332 0 0.6047>, rot=<0 0 0.0181 -0.9998>}
 dyn_poncho_back_l_05 {parent=dyn_poncho_back_l_04, loc=<11.5157 0 0>, rot=<0 -0.7071 0.7071 0>}
 thigh_l {parent=pelvis, loc=<-2.5693 -0.0109 -7.7501>, rot=<0.0012 0.0572 0.0155 -0.9982>}
 calf_l {parent=thigh_l, loc=<-39.0484 0 0>, rot=<0.0006 -0.0161 -0.0391 -0.9991>}
 calf_twist_01_l {parent=calf_l, loc=<-35.813 -0.0967 0.0397>, rot=<0.0237 0.0095 -0.0035 -0.9997>}
 calf_twist_02_l {parent=calf_l, loc=<-18.8064 -0.0967 0.0397>, rot=<0.0237 0.0095 -0.0035 -0.9997>}
 foot_l {parent=calf_l, loc=<-42.561 0.1077 0.4292>, rot=<0.0418 -0.0455 0.0343 -0.9975>}
 ball_l {parent=foot_l, loc=<-6.1599 13.8264 0.6806>, rot=<0.0001 -0.0001 -0.7063 -0.7079>}
 dyn_l_shoelaces_master {parent=calf_l, loc=<-6.6044 5.0582 -0.6586>, rot=<-0.0221 -0.7694 -0.6379 -0.0241>}
 dyn_l_knee_loop_l_01 {parent=dyn_l_shoelaces_master, loc=<0.5893 0.0291 -0.1333>, rot=<0.0847 0.0378 -0.1994 -0.9755>}
 dyn_l_knee_loop_l_02 {parent=dyn_l_knee_loop_l_01, loc=<2.673 0 0>, rot=<0.0011 -0.0118 0.0125 -0.9999>}
 dyn_l_knee_string_r_loop_01 {parent=dyn_l_shoelaces_master, loc=<0.2666 -0.2604 -0.0839>, rot=<0.0918 0.0231 0.1486 -0.9843>}
 dyn_l_knee_string_r_loop_02 {parent=dyn_l_knee_string_r_loop_01, loc=<1.9082 0 0>, rot=<0.0033 -0.0532 -0.0616 -0.9967>}
 dyn_l_knee_string_r_loop_03 {parent=dyn_l_knee_string_r_loop_02, loc=<2.5092 0 0>, rot=<0.0016 0.0532 -0.0302 -0.9981>}
 dyn_l_knee_string_long_master {parent=calf_l, loc=<-7.0654 4.853 -0.3722>, rot=<0.0646 0.7031 0.7081 -0.0061>}
 dyn_l_knee_string_long_01 {parent=dyn_l_knee_string_long_master, loc=<0 0 0>, rot=<0 0 0 1>}
 dyn_l_knee_string_long_02 {parent=dyn_l_knee_string_long_01, loc=<5.7214 0 0>, rot=<0 0 0.0104 -0.9999>}
 dyn_l_knee_string_long_03 {parent=dyn_l_knee_string_long_02, loc=<5.7547 0 0>, rot=<0.0001 -0.0184 -0.0039 -0.9998>}
 thigh_twist_01_l {parent=thigh_l, loc=<-21.6303 0 0>, rot=<0.0013 0 0 -1>}
 thigh_r {parent=pelvis, loc=<-2.5691 -0.0109 7.7501>, rot=<0.0572 -0.0012 0.9982 0.0155>}
 calf_r {parent=thigh_r, loc=<39.0487 0 0>, rot=<0.0006 -0.0161 -0.0391 -0.9991>}
 calf_twist_01_r {parent=calf_r, loc=<35.8132 0.0967 -0.0397>, rot=<0.0237 0.0095 -0.0035 -0.9997>}
 calf_twist_02_r {parent=calf_r, loc=<18.8064 0.0967 -0.0397>, rot=<0.0237 0.0095 -0.0035 -0.9997>}
 foot_r {parent=calf_r, loc=<42.5612 -0.1077 -0.4292>, rot=<0.0418 -0.0455 0.0343 -0.9975>}
 ball_r {parent=foot_r, loc=<6.1598 -13.8264 -0.6806>, rot=<0.0001 -0.0001 -0.7063 -0.7079>}
 thigh_twist_01_r {parent=thigh_r, loc=<21.6303 0 0>, rot=<0.0013 0 0 -1>}
 dyn_r_shoelaces_master {parent=thigh_r, loc=<31.405 -7.6941 1.1674>, rot=<-0.7043 0.0099 0.0268 -0.7093>}
 dyn_r_knee_string_01 {parent=dyn_r_shoelaces_master, loc=<0.6711 0.0847 -0.0351>, rot=<-0.0034 0 0 -1>}
 dyn_r_knee_string_02 {parent=dyn_r_knee_string_01, loc=<3.3286 0 0>, rot=<0.0001 0.0046 0.002 -1>}
 dyn_r_knee_string_03 {parent=dyn_r_knee_string_02, loc=<4.8725 0 0>, rot=<0.0011 -0.0071 0.0222 -0.9997>}
 dyn_r_knee_string_r_01 {parent=dyn_r_shoelaces_master, loc=<0.4741 0.6166 -0.1012>, rot=<0.0087 -0.0624 -0.4644 -0.8834>}
 dyn_r_string_r_01 {parent=dyn_r_shoelaces_master, loc=<0.5749 -0.1043 -0.1108>, rot=<0.0066 0.0105 0.1693 -0.9855>}
 dyn_r_string_r_02 {parent=dyn_r_string_r_01, loc=<1.884 0 0>, rot=<-0.0013 -0.007 -0.0212 -0.9997>}
 dyn_belt_knife_r_master {parent=pelvis, loc=<-1.4374 0.7591 16.2113>, rot=<0.0907 -0.0016 0.9957 0.0178>}
 dyn_belt_knife_r_01 {parent=dyn_belt_knife_r_master, loc=<0 0 0>, rot=<0 0 0 1>}
 dyn_belt_knife_r_02 {parent=dyn_belt_knife_r_01, loc=<4.3801 0 0>, rot=<0 -0.0492 0 -0.9988>}
 dyn_belt_knife_r_03 {parent=dyn_belt_knife_r_02, loc=<9.5283 0 0>, rot=<0 -0.0246 0 -0.9997>}
 dyn_belt_knife_l_master {parent=pelvis, loc=<-1.4375 0.7591 -16.2113>, rot=<-0.0178 -0.9957 0.0016 -0.0907>}
 dyn_belt_knife_l_01 {parent=dyn_belt_knife_l_master, loc=<0 0 0>, rot=<0 0 0 1>}
 dyn_belt_knife_l_02 {parent=dyn_belt_knife_l_01, loc=<4.3801 0 0>, rot=<0 -0.0492 0 -0.9988>}
 dyn_belt_knife_l_03 {parent=dyn_belt_knife_l_02, loc=<9.5283 0 0>, rot=<0 -0.0246 0 -0.9997>}
 dyn_back_strap_master {parent=pelvis, loc=<0.5158 -9.2939 12.5066>, rot=<0.0252 -0.3877 0.9209 -0.0327>}
 dyn_back_strap_01 {parent=dyn_back_strap_master, loc=<0 0 0>, rot=<0 0 0 1>}
 dyn_back_strap_02 {parent=dyn_back_strap_01, loc=<5.1138 0 0>, rot=<0.0414 -0.0231 -0.0115 -0.9988>}
 dyn_back_strap_03 {parent=dyn_back_strap_02, loc=<5.3129 -0.2076 -0.0067>, rot=<0.0102 -0.0421 0.033 -0.9985>}
]
