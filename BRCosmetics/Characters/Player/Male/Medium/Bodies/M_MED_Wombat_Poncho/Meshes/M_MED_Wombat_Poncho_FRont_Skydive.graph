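# Physics asset for the front of the medium male Wombat poncho cosmetic, named as the skydive variant.
# Capsules cover pelvis, spine, upper arms, legs and belt knives, while boxes form the poncho, sleeve and armpit bodies.
# Five four-segment front chains hang from spine_04 with tight twist limits, cross-linked at the second and fourth rows by angularly free constraints to act as a sheet.
# Sleeves attach to upper arms, armpit pieces to spine_02.
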PhysicsAsset /BRCosmetics/Characters/Player/Male/Medium/Bodies/M_MED_Wombat_Poncho/Meshes/M_MED_Wombat_Poncho_FRont_Skydive
bodies {
 calf_l [{shapeType=Capsule, length=29.0991, radius=7.9827, center=<-17.4253 0.0004 0>, rotation=<rot 90 0 0>}]
 calf_r [{shapeType=Capsule, length=29.3558, radius=8.1072, center=<17.5455 0.0004 0>, rotation=<rot 90 0 0>}]
 dyn_belt_knife_l_master [{shapeType=Capsule, length=13.7559, radius=6.2648, center=<9.3706 0 -0.0002>, rotation=<rot 90 0 0>}]
 dyn_belt_knife_r_master [{shapeType=Capsule, length=12.6221, radius=7.3986, center=<9.3706 0 -0.0002>, rotation=<rot 90 0 0>}]
 dyn_l_sleeve_front_01 [{shapeType=Box, extentX=11.7834, extentY=7.8811, extentZ=2, center=<6.0335 0 0>}]
 dyn_poncho_armpit_front_l [{shapeType=Box, extentX=3.5, extentY=3.5, extentZ=3.5, center=<-2.9846 0 0>}]
 dyn_poncho_armpit_front_r [{shapeType=Box, extentX=3.5, extentY=3.5, extentZ=3.5, center=<1.7404 0 0>}]
 dyn_poncho_front_l_01 [{shapeType=Box, extentX=17.3302, extentY=12.8255, extentZ=5, center=<8.5802 0 -0.0006>}]
 dyn_poncho_front_l_02 [{shapeType=Box, extentX=17.3302, extentY=14.3812, extentZ=5, center=<8.5802 0 -0.0006>}]
 dyn_poncho_front_l_03 [{shapeType=Box, extentX=17.3302, extentY=14.4832, extentZ=5, center=<8.5802 0 -0.0006>}]
 dyn_poncho_front_l_04 [{shapeType=Box, extentX=17.3302, extentY=13.7008, extentZ=3.6014, center=<8.5802 0 -0.0006>}]
 dyn_poncho_front_mid_01 [{shapeType=Box, extentX=17.3302, extentY=12.8255, extentZ=5, center=<8.5802 0 -0.0006>}]
 dyn_poncho_front_mid_02 [{shapeType=Box, extentX=17.3302, extentY=14.3812, extentZ=5, center=<8.5802 0 -0.0006>}]
 dyn_poncho_front_mid_03 [{shapeType=Box, extentX=17.3302, extentY=14.4832, extentZ=5, center=<8.5802 0 -0.0006>}]
 dyn_poncho_front_mid_04 [{shapeType=Box, extentX=18.7103, extentY=13.7008, extentZ=3.6014, center=<9.8519 0 -0.0006>}]
 dyn_poncho_front_ml_01 [{shapeType=Box, extentX=17.3302, extentY=12.8255, extentZ=5, center=<8.5802 0 -0.0006>}]
 dyn_poncho_front_ml_02 [{shapeType=Box, extentX=17.3302, extentY=14.3812, extentZ=5, center=<8.5802 0 -0.0006>}]
 dyn_poncho_front_ml_03 [{shapeType=Box, extentX=17.3302, extentY=14.4832, extentZ=5, center=<8.5802 0 -0.0006>}]
 dyn_poncho_front_ml_04 [{shapeType=Box, extentX=17.3302, extentY=13.7008, extentZ=3.6014, center=<8.5802 0 -0.0006>}]
 dyn_poncho_front_mr_01 [{shapeType=Box, extentX=17.3302, extentY=12.8255, extentZ=5, center=<8.5802 0 -0.0006>}]
 dyn_poncho_front_mr_02 [{shapeType=Box, extentX=17.3302, extentY=14.3812, extentZ=5, center=<8.5802 0 -0.0006>}]
 dyn_poncho_front_mr_03 [{shapeType=Box, extentX=17.3302, extentY=14.4832, extentZ=5, center=<8.5802 0 -0.0006>}]
 dyn_poncho_front_mr_04 [{shapeType=Box, extentX=17.3302, extentY=13.7008, extentZ=3.6014, center=<8.5802 0 -0.0006>}]
 dyn_poncho_front_r_01 [{shapeType=Box, extentX=17.3302, extentY=12.8255, extentZ=5, center=<8.5802 0 -0.0006>}]
 dyn_poncho_front_r_02 [{shapeType=Box, extentX=17.3302, extentY=14.3812, extentZ=5, center=<8.5802 0 -0.0006>}]
 dyn_poncho_front_r_03 [{shapeType=Box, extentX=17.3302, extentY=14.4832, extentZ=5, center=<8.5802 0 -0.0006>}]
 dyn_poncho_front_r_04 [{shapeType=Box, extentX=17.3302, extentY=13.7008, extentZ=3.6014, center=<8.5802 0 -0.0006>}]
 dyn_r_sleeve_front_01 [{shapeType=Box, extentX=11.7834, extentY=7.8811, extentZ=2, center=<6.0335 0 0>}]
 pelvis [
  {shapeType=Capsule, length=27.303, radius=8.8399, center=<0 -1.0033 0>}
  {shapeType=Capsule, length=27.1877, radius=8.9552, center=<-7.3932 -4.7646 0>}
 ]
 spine_02 [{shapeType=Capsule, length=28.9038, radius=8.8937, center=<0 -1.0267 0>}]
 spine_04 [{shapeType=Capsule, length=34.6405, radius=7.6599, center=<0 -0.7203 0>}]
 spine_05 [{shapeType=Capsule, length=28.1429, radius=8}]
 thigh_l [{shapeType=Capsule, length=27.1867, radius=9.6603, center=<-19.4433 0.0005 0>, rotation=<rot 90 0 0>}]
 thigh_r [{shapeType=Capsule, length=27.1867, radius=9.6603, center=<19.3925 0.0011 0>, rotation=<rot 90 0 0>}]
 upperarm_l [{shapeType=Capsule, length=15.9526, radius=5.7315, center=<20.2981 0.0002 0>, rotation=<rot 83.5897 -180 -180>}]
 upperarm_r [{shapeType=Capsule, length=12.5107, radius=6.1455, center=<-18.3602 -0.0001 0.5758>, rotation=<rot 87.5622 -179.9999 -179.9999>}]
}
constraints [
 {bone1Name=spine_04, bone2Name=spine_02, swing1LimitDegrees=45, swing2LimitDegrees=45, twistLimitDegrees=45}
 {bone1Name=spine_05, bone2Name=spine_04, swing1LimitDegrees=45, swing2LimitDegrees=45, twistLimitDegrees=45}
 {bone1Name=upperarm_l, bone2Name=spine_05, swing1LimitDegrees=45, swing2LimitDegrees=45, twistLimitDegrees=45}
 {bone1Name=upperarm_r, bone2Name=spine_05, swing1LimitDegrees=45, swing2LimitDegrees=45, twistLimitDegrees=45}
 {
  bone1Name=spine_02
  bone2Name=pelvis
  swing1LimitDegrees=45
  swing1Motion=Free
  swing2LimitDegrees=45
  swing2Motion=Free
  twistLimitDegrees=45
  twistMotion=Free
 }
 {
  bone1Name=thigh_l
  bone2Name=pelvis
  swing1LimitDegrees=45
  swing1Motion=Free
  swing2LimitDegrees=45
  swing2Motion=Free
  twistLimitDegrees=45
  twistMotion=Free
 }
 {
  bone1Name=thigh_r
  bone2Name=pelvis
  swing1LimitDegrees=45
  swing1Motion=Free
  swing2LimitDegrees=45
  swing2Motion=Free
  twistLimitDegrees=45
  twistMotion=Free
 }
 {bone1Name=dyn_poncho_front_mr_01, bone2Name=spine_04, swing1LimitDegrees=45, swing2LimitDegrees=45, twistLimitDegrees=10}
 {bone1Name=dyn_poncho_front_mr_02, bone2Name=dyn_poncho_front_mr_01, swing1LimitDegrees=45, swing2LimitDegrees=45, twistLimitDegrees=10}
 {bone1Name=dyn_poncho_front_mr_03, bone2Name=dyn_poncho_front_mr_02, swing1LimitDegrees=45, swing2LimitDegrees=45, twistLimitDegrees=10}
 {bone1Name=dyn_poncho_front_mr_04, bone2Name=dyn_poncho_front_mr_03, swing1LimitDegrees=45, swing2LimitDegrees=45, twistLimitDegrees=10}
 {bone1Name=dyn_poncho_front_ml_01, bone2Name=spine_04, swing1LimitDegrees=45, swing2LimitDegrees=45, twistLimitDegrees=10}
 {bone1Name=dyn_poncho_front_ml_02, bone2Name=dyn_poncho_front_ml_01, swing1LimitDegrees=45, swing2LimitDegrees=45, twistLimitDegrees=10}
 {bone1Name=dyn_poncho_front_ml_03, bone2Name=dyn_poncho_front_ml_02, swing1LimitDegrees=45, swing2LimitDegrees=45, twistLimitDegrees=10}
 {bone1Name=dyn_poncho_front_ml_04, bone2Name=dyn_poncho_front_ml_03, swing1LimitDegrees=45, swing2LimitDegrees=45, twistLimitDegrees=10}
 {bone1Name=dyn_poncho_front_mid_01, bone2Name=spine_04, swing1LimitDegrees=45, swing2LimitDegrees=45, twistLimitDegrees=10}
 {bone1Name=dyn_poncho_front_mid_02, bone2Name=dyn_poncho_front_mid_01, swing1LimitDegrees=45, swing2LimitDegrees=45, twistLimitDegrees=10}
 {bone1Name=dyn_poncho_front_mid_03, bone2Name=dyn_poncho_front_mid_02, swing1LimitDegrees=45, swing2LimitDegrees=45, twistLimitDegrees=10}
 {bone1Name=dyn_poncho_front_mid_04, bone2Name=dyn_poncho_front_mid_03, swing1LimitDegrees=45, swing2LimitDegrees=45, twistLimitDegrees=10}
 {bone1Name=dyn_poncho_front_l_01, bone2Name=spine_04, swing1LimitDegrees=45, swing2LimitDegrees=45, twistLimitDegrees=10}
 {bone1Name=dyn_poncho_front_l_02, bone2Name=dyn_poncho_front_l_01, swing1LimitDegrees=45, swing2LimitDegrees=45, twistLimitDegrees=10}
 {bone1Name=dyn_poncho_front_l_03, bone2Name=dyn_poncho_front_l_02, swing1LimitDegrees=45, swing2LimitDegrees=45, twistLimitDegrees=10}
 {bone1Name=dyn_poncho_front_l_04, bone2Name=dyn_poncho_front_l_03, swing1LimitDegrees=45, swing2LimitDegrees=45, twistLimitDegrees=10}
 {bone1Name=dyn_poncho_front_r_01, bone2Name=spine_04, swing1LimitDegrees=45, swing2LimitDegrees=45, twistLimitDegrees=10}
 {bone1Name=dyn_poncho_front_r_02, bone2Name=dyn_poncho_front_r_01, swing1LimitDegrees=45, swing2LimitDegrees=45, twistLimitDegrees=10}
 {bone1Name=dyn_poncho_front_r_03, bone2Name=dyn_poncho_front_r_02, swing1LimitDegrees=45, swing2LimitDegrees=45, twistLimitDegrees=10}
 {bone1Name=dyn_poncho_front_r_04, bone2Name=dyn_poncho_front_r_03, swing1LimitDegrees=45, swing2LimitDegrees=45, twistLimitDegrees=10}
 {bone1Name=dyn_l_sleeve_front_01, bone2Name=upperarm_l, swing1LimitDegrees=35, swing2LimitDegrees=35, twistLimitDegrees=20}
 {bone1Name=dyn_r_sleeve_front_01, bone2Name=upperarm_r, swing1LimitDegrees=35, swing2LimitDegrees=35, twistLimitDegrees=20}
 {
  bone1Name=dyn_poncho_front_r_02
  bone2Name=dyn_poncho_front_mr_02
  swing1LimitDegrees=45
  swing1Motion=Free
  swing2LimitDegrees=45
  swing2Motion=Free
  twistLimitDegrees=45
  twistMotion=Free
 }
 {
  bone1Name=dyn_poncho_front_mid_02
  bone2Name=dyn_poncho_front_mr_02
  swing1LimitDegrees=45
  swing1Motion=Free
  swing2LimitDegrees=45
  swing2Motion=Free
  twistLimitDegrees=45
  twistMotion=Free
 }
 {
  bone1Name=dyn_poncho_front_l_02
  bone2Name=dyn_poncho_front_ml_02
  swing1LimitDegrees=45
  swing1Motion=Free
  swing2LimitDegrees=45
  swing2Motion=Free
  twistLimitDegrees=45
  twistMotion=Free
 }
 {
  bone1Name=dyn_poncho_front_mid_02
  bone2Name=dyn_poncho_front_ml_02
  swing1LimitDegrees=45
  swing1Motion=Free
  swing2LimitDegrees=45
  swing2Motion=Free
  twistLimitDegrees=45
  twistMotion=Free
 }
 {
  bone1Name=dyn_poncho_front_r_04
  bone2Name=dyn_poncho_front_mr_04
  swing1LimitDegrees=45
  swing1Motion=Free
  swing2LimitDegrees=45
  swing2Motion=Free
  twistLimitDegrees=45
  twistMotion=Free
 }
 {
  bone1Name=dyn_poncho_front_mid_04
  bone2Name=dyn_poncho_front_mr_04
  swing1LimitDegrees=45
  swing1Motion=Free
  swing2LimitDegrees=45
  swing2Motion=Free
  twistLimitDegrees=45
  twistMotion=Free
 }
 {
  bone1Name=dyn_poncho_front_l_04
  bone2Name=dyn_poncho_front_ml_04
  swing1LimitDegrees=45
  swing1Motion=Free
  swing2LimitDegrees=45
  swing2Motion=Free
  twistLimitDegrees=45
  twistMotion=Free
 }
 {
  bone1Name=dyn_poncho_front_mid_04
  bone2Name=dyn_poncho_front_ml_04
  swing1LimitDegrees=45
  swing1Motion=Free
  swing2LimitDegrees=45
  swing2Motion=Free
  twistLimitDegrees=45
  twistMotion=Free
 }
 {bone1Name=dyn_poncho_armpit_front_r, bone2Name=spine_02, swing1LimitDegrees=25, swing2LimitDegrees=25, twistLimitDegrees=20}
 {bone1Name=dyn_poncho_armpit_front_l, bone2Name=spine_02, swing1LimitDegrees=25, swing2LimitDegrees=25, twistLimitDegrees=20}
 {
  bone1Name=calf_r
  bone2Name=thigh_r
  swing1LimitDegrees=45
  swing1Motion=Free
  swing2LimitDegrees=45
  swing2Motion=Free
  twistLimitDegrees=45
  twistMotion=Free
 }
 {
  bone1Name=calf_l
  bone2Name=thigh_l
  swing1LimitDegrees=45
  swing1Motion=Free
  swing2LimitDegrees=45
  swing2Motion=Free
  twistLimitDegrees=45
  twistMotion=Free
 }
 {
  bone1Name=dyn_belt_knife_r_master
  bone2Name=pelvis
  swing1LimitDegrees=45
  swing1Motion=Free
  swing2LimitDegrees=45
  swing2Motion=Free
  twistLimitDegrees=45
  twistMotion=Free
 }
 {
  bone1Name=dyn_belt_knife_l_master
  bone2Name=pelvis
  swing1LimitDegrees=45
  swing1Motion=Free
  swing2LimitDegrees=45
  swing2Motion=Free
  twistLimitDegrees=45
  twistMotion=Free
 }
]
