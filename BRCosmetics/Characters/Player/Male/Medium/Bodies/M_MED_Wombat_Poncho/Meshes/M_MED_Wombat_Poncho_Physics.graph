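# Physics asset for the male medium Wombat Poncho cosmetic body, setting up collision and constraints for the poncho's dynamic bones.
# Capsules on spine, pelvis, upper arms, legs and belt knives form the body volume.
# Five front and five back four-bone chains hang from spine_04, the front via small master boxes; sleeves attach to upper arms, armpits to spine_02.
# Angularly free cross-links at the second and fourth rows tie neighbouring chains into cloth-like panels.
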PhysicsAsset /BRCosmetics/Characters/Player/Male/Medium/Bodies/M_MED_Wombat_Poncho/Meshes/M_MED_Wombat_Poncho_Physics
bodies {
 calf_l [{shapeType=Capsule, length=29.0991, radius=7.9827, center=<-17.4253 0.0004 0>, rotation=<rot 90 0 0>}]
 calf_r [{shapeType=Capsule, length=29.3558, radius=8.1072, center=<17.5455 0.0004 0>, rotation=<rot 90 0 0>}]
 dyn_belt_knife_l_master [{shapeType=Capsule, length=13.7559, radius=6.2648, center=<9.3706 0 -0.0002>, rotation=<rot 90 0 0>}]
 dyn_belt_knife_r_master [{shapeType=Capsule, length=12.6221, radius=7.3986, center=<9.3706 0 -0.0002>, rotation=<rot 90 0 0>}]
 dyn_l_sleeve_back_01 [{shapeType=Capsule, length=8.5559, radius=4.0094, shapeName=dyn_r_sleeve_back_01_capsule, center=<5.3425 0 2.8139>, rotation=<rot 90 0 0>}]
 dyn_l_sleeve_front_01 [{shapeType=Capsule, length=8.5559, radius=4.0094, shapeName=dyn_r_sleeve_back_01_capsule, center=<5.3425 0 2.8139>, rotation=<rot 90 0 0>}]
 dyn_l_sleeve_mid_01 [{
   shapeType=Capsule
   length=8.5559
   radius=4.0094
   shapeName=dyn_r_sleeve_back_01_capsule
   center=<5.3669 -0.0007 3.5313>
   rotation=<rot 84.2467 -0.66 -0.7186>
  }]
 dyn_poncho_armpit_back_l [{shapeType=Box, extentX=3.5, extentY=3.5, extentZ=3.5, center=<-2.9846 0 0>}]
 dyn_poncho_armpit_back_r [{shapeType=Box, extentX=3.5, extentY=3.5, extentZ=3.5, center=<1.7404 0 0>}]
 dyn_poncho_armpit_front_l [{shapeType=Box, extentX=3.5, extentY=3.5, extentZ=3.5, center=<-2.9846 0 0>}]
 dyn_poncho_armpit_front_r [{shapeType=Box, extentX=3.5, extentY=3.5, extentZ=3.5, center=<1.7404 0 0>}]
 dyn_poncho_back_l_01 [{shapeType=Box, extentX=17.3302, extentY=12.8255, extentZ=3.2177, center=<8.3604 0 3.4981>, rotation=<rot 13.6767 0 0>}]
 dyn_poncho_back_l_02 [{shapeType=Box, extentX=17.3302, extentY=14.3812, extentZ=3.7335, center=<8.5802 0 -0.0006>}]
 dyn_poncho_back_l_03 [{shapeType=Box, extentX=17.3302, extentY=14.4832, extentZ=3.7593, center=<8.5802 0 -0.0006>}]
 dyn_poncho_back_l_04 [{shapeType=Box, extentX=17.3302, extentY=13.7008, extentZ=5, center=<8.5802 0 -0.0006>}]
 dyn_poncho_back_mid_01 [{shapeType=Box, extentX=17.3302, extentY=12.8255, extentZ=8.3754, center=<8.5802 0 -0.0006>}]
 dyn_poncho_back_mid_02 [{shapeType=Box, extentX=17.3302, extentY=14.3812, extentZ=5.9923, center=<8.5802 0 -0.0006>}]
 dyn_poncho_back_mid_03 [{shapeType=Capsule, length=7.056, radius=5.5093, shapeName=dyn_r_sleeve_back_01_capsule, center=<6.786 0 0>, rotation=<rot 90 0 0>}]
 dyn_poncho_back_mid_04 [{shapeType=Box, extentX=17.3302, extentY=13.7008, extentZ=5, center=<8.5802 0 -0.0006>}]
 dyn_poncho_back_ml_01 [{shapeType=Box, extentX=17.3302, extentY=12.8255, extentZ=6.2675, center=<8.5802 0 1.1368>}]
 dyn_poncho_back_ml_02 [{shapeType=Box, extentX=17.3302, extentY=14.3812, extentZ=6.2809, center=<8.5802 0 -0.0006>}]
 dyn_poncho_back_ml_03 [{shapeType=Capsule, length=7.056, radius=5.5093, shapeName=dyn_r_sleeve_back_01_capsule, center=<6.786 0 0>, rotation=<rot 90 0 0>}]
 dyn_poncho_back_ml_04 [{shapeType=Box, extentX=17.3302, extentY=13.7008, extentZ=5, center=<8.5802 0 -0.0006>}]
 dyn_poncho_back_mr_01 [{shapeType=Box, extentX=17.3302, extentY=12.8255, extentZ=5.6382, center=<8.5802 0 0.8118>}]
 dyn_poncho_back_mr_02 [{shapeType=Box, extentX=17.3302, extentY=14.3812, extentZ=6.2809, center=<8.5802 0 -0.0006>}]
 dyn_poncho_back_mr_03 [{shapeType=Capsule, length=7.056, radius=5.5093, shapeName=dyn_r_sleeve_back_01_capsule, center=<6.786 0 0>, rotation=<rot 90 0 0>}]
 dyn_poncho_back_mr_04 [{shapeType=Box, extentX=17.3302, extentY=13.7008, extentZ=5, center=<8.5802 0 -0.0006>}]
 dyn_poncho_back_r_01 [{shapeType=Box, extentX=17.3302, extentY=12.8255, extentZ=3.2177, center=<7.897 0 2.9532>, rotation=<rot 13.0226 0 0>}]
 dyn_poncho_back_r_02 [{shapeType=Box, extentX=17.3302, extentY=14.3812, extentZ=3.7335, center=<8.0284 0 1.6046>, rotation=<rot 10.1847 0 0>}]
 dyn_poncho_back_r_03 [{shapeType=Box, extentX=17.3302, extentY=14.4832, extentZ=3.7593, center=<8.5802 0 -0.0006>}]
 dyn_poncho_back_r_04 [{shapeType=Box, extentX=17.3302, extentY=13.7008, extentZ=5, center=<8.5802 0 -0.0006>}]
 dyn_poncho_front_l_01 [{shapeType=Box, extentX=17.3302, extentY=12.8255, extentZ=5, center=<8.5802 0 -0.0006>}]
 dyn_poncho_front_l_02 [{shapeType=Box, extentX=17.3302, extentY=14.3812, extentZ=5, center=<8.5802 0 -0.0006>}]
 dyn_poncho_front_l_03 [{shapeType=Box, extentX=17.3302, extentY=14.4832, extentZ=5, center=<8.5802 0 -0.0006>}]
 dyn_poncho_front_l_04 [{shapeType=Capsule, length=8.5559, radius=4.0094, shapeName=dyn_r_sleeve_back_01_capsule, center=<5.3425 0 2.8139>, rotation=<rot 90 0 0>}]
 dyn_poncho_front_l_master [{shapeType=Box, extentX=1.01, extentY=1.01, extentZ=1.01, rotation=<rot 0 -90 0>}]
 dyn_poncho_front_mid_01 [{shapeType=Capsule, length=6.9726, radius=5.5833, shapeName=dyn_r_sleeve_back_01_capsule, center=<5.3425 0 2.8139>, rotation=<rot 90 0 0>}]
 dyn_poncho_front_mid_02 [{shapeType=Capsule, length=15.9726, radius=6.0833, shapeName=dyn_r_sleeve_back_01_capsule, center=<5.3425 0 -0.1263>, rotation=<rot 90 0 0>}]
 dyn_poncho_front_mid_03 [{shapeType=Capsule, length=6.9726, radius=5.5833, shapeName=dyn_r_sleeve_back_01_capsule, center=<5.3425 0 2.8139>, rotation=<rot 90 0 0>}]
 dyn_poncho_front_mid_04 [{shapeType=Capsule, length=7.6065, radius=4.9588, shapeName=dyn_r_sleeve_back_01_capsule, center=<8.2585 0 2.8139>, rotation=<rot 90 0 0>}]
 dyn_poncho_front_mid_master [{shapeType=Box, extentX=1.01, extentY=1.01, extentZ=1.01, rotation=<rot 0 -90 0>}]
 dyn_poncho_front_ml_01 [{shapeType=Capsule, length=6.9726, radius=5.5833, shapeName=dyn_r_sleeve_back_01_capsule, center=<5.3425 0 2.8139>, rotation=<rot 90 0 0>}]
 dyn_poncho_front_ml_02 [{shapeType=Capsule, length=6.968, radius=5.5879, shapeName=dyn_r_sleeve_back_01_capsule, center=<5.3425 0 2.8139>, rotation=<rot 90 0 0>}]
 dyn_poncho_front_ml_03 [{shapeType=Capsule, length=6.9726, radius=5.5927, shapeName=dyn_r_sleeve_back_01_capsule, center=<5.3425 0 2.8139>, rotation=<rot 90 0 0>}]
 dyn_poncho_front_ml_04 [{shapeType=Capsule, length=7.6065, radius=4.9588, shapeName=dyn_r_sleeve_back_01_capsule, center=<8.2585 0 2.8139>, rotation=<rot 90 0 0>}]
 dyn_poncho_front_ml_master [{shapeType=Box, extentX=1.01, extentY=1.01, extentZ=1.01, rotation=<rot 0 -90 0>}]
 dyn_poncho_front_mr_01 [{shapeType=Capsule, length=6.9726, radius=5.5833, shapeName=dyn_r_sleeve_back_01_capsule, center=<5.3425 0 2.8139>, rotation=<rot 90 0 0>}]
 dyn_poncho_front_mr_02 [{shapeType=Capsule, length=6.9726, radius=5.5833, shapeName=dyn_r_sleeve_back_01_capsule, center=<5.3425 0 2.8139>, rotation=<rot 90 0 0>}]
 dyn_poncho_front_mr_03 [{shapeType=Capsule, length=6.9726, radius=5.5833, shapeName=dyn_r_sleeve_back_01_capsule, center=<5.3425 0 2.8139>, rotation=<rot 90 0 0>}]
 dyn_poncho_front_mr_04 [{shapeType=Capsule, length=7.6065, radius=4.9588, shapeName=dyn_r_sleeve_back_01_capsule, center=<8.2585 0 2.8139>, rotation=<rot 90 0 0>}]
 dyn_poncho_front_mr_master [{shapeType=Box, extentX=1.01, extentY=1.01, extentZ=1.01, rotation=<rot 0 -90 0>}]
 dyn_poncho_front_r_01 [{shapeType=Box, extentX=17.3302, extentY=12.8255, extentZ=5, center=<8.5802 0 -0.0006>}]
 dyn_poncho_front_r_02 [{shapeType=Box, extentX=17.3302, extentY=14.3812, extentZ=5, center=<8.5802 0 -0.0006>}]
 dyn_poncho_front_r_03 [{shapeType=Box, extentX=17.3302, extentY=14.4832, extentZ=5, center=<8.5802 0 -0.0006>}]
 dyn_poncho_front_r_04 [{shapeType=Capsule, length=8.5559, radius=4.0094, shapeName=dyn_r_sleeve_back_01_capsule, center=<5.3425 0 2.8139>, rotation=<rot 90 0 0>}]
 dyn_poncho_front_r_master [{shapeType=Box, extentX=1.01, extentY=1.01, extentZ=1.01, rotation=<rot 0 -90 0>}]
 dyn_r_sleeve_back_01 [{shapeType=Capsule, length=8.5559, radius=4.0094, shapeName=dyn_r_sleeve_back_01_capsule, center=<5.3425 0 2.8139>, rotation=<rot 90 0 0>}]
 dyn_r_sleeve_front_01 [{shapeType=Capsule, length=8.5559, radius=4.0094, shapeName=dyn_r_sleeve_back_01_capsule, center=<5.3425 0 2.8139>, rotation=<rot 90 0 0>}]
 dyn_r_sleeve_mid_01 [{
   shapeType=Capsule
   length=8.5559
   radius=4.0094
   shapeName=dyn_r_sleeve_back_01_capsule
   center=<5.3458 0 3.0659>
   rotation=<rot 84.0684 0.0018 0.0019>
  }]
 pelvis [
  {shapeType=Capsule, length=27.303, radius=8.8399, center=<0 -1.0033 0>}
  {shapeType=Capsule, length=27.1877, radius=8.9552, center=<-7.3932 -4.7646 0>}
 ]
 spine_02 [{shapeType=Capsule, length=26.86, radius=10.9375, center=<0 -1.0267 0>}]
 spine_04 [{shapeType=Capsule, length=32.8339, radius=9.4665, center=<0 -0.7203 0>}]
 spine_05 [{shapeType=Capsule, length=28.1429, radius=8}]
 thigh_l [{shapeType=Capsule, length=27.1867, radius=9.6603, center=<-19.4433 0.0005 0>, rotation=<rot 90 0 0>}]
 thigh_r [{shapeType=Capsule, length=27.1867, radius=9.6603, center=<19.3925 0.0011 0>, rotation=<rot 90 0 0>}]
 upperarm_l [{shapeType=Capsule, length=15.9526, radius=5.7315, center=<20.2981 0.0002 0>, rotation=<rot 83.5897 -180 -180>}]
 upperarm_r [{shapeType=Capsule, length=12.5107, radius=6.1455, center=<-18.3602 -0.0001 0.5758>, rotation=<rot 87.5622 -179.9999 -179.9999>}]
}
constraints [
 {bone1Name=spine_04, bone2Name=spine_02, swing1LimitDegrees=45, swing2LimitDegrees=45, twistLimitDegrees=45}
 {bone1Name=spine_05, bone2Name=spine_04, swing1LimitDegrees=45, swing2LimitDegrees=45, twistLimitDegrees=45}
 {bone1Name=upperarm_l, bone2Name=spine_05, swing1LimitDegrees=45, swing2LimitDegrees=45, twistLimitDegrees=45}
 {bone1Name=upperarm_r, bone2Name=spine_05, swing1LimitDegrees=45, swing2LimitDegrees=45, twistLimitDegrees=45}
 {
  bone1Name=spine_02
  bone2Name=pelvis
  swing1LimitDegrees=45
  swing1Motion=Free
  swing2LimitDegrees=45
  swing2Motion=Free
  twistLimitDegrees=45
  twistMotion=Free
 }
 {
  bone1Name=thigh_l
  bone2Name=pelvis
  swing1LimitDegrees=45
  swing1Motion=Free
  swing2LimitDegrees=45
  swing2Motion=Free
  twistLimitDegrees=45
  twistMotion=Free
 }
 {
  bone1Name=thigh_r
  bone2Name=pelvis
  swing1LimitDegrees=45
  swing1Motion=Free
  swing2LimitDegrees=45
  swing2Motion=Free
  twistLimitDegrees=45
  twistMotion=Free
 }
 {bone1Name=dyn_poncho_front_mr_01, bone2Name=dyn_poncho_front_mr_master, swing1LimitDegrees=45, swing2LimitDegrees=45, twistLimitDegrees=10}
 {bone1Name=dyn_poncho_front_mr_02, bone2Name=dyn_poncho_front_mr_01, swing1LimitDegrees=45, swing2LimitDegrees=45, twistLimitDegrees=10}
 {bone1Name=dyn_poncho_front_mr_03, bone2Name=dyn_poncho_front_mr_02, swing1LimitDegrees=45, swing2LimitDegrees=45, twistLimitDegrees=10}
 {bone1Name=dyn_poncho_front_mr_04, bone2Name=dyn_poncho_front_mr_03, swing1LimitDegrees=45, swing2LimitDegrees=45, twistLimitDegrees=10}
 {bone1Name=dyn_poncho_front_ml_01, bone2Name=dyn_poncho_front_ml_master, swing1LimitDegrees=45, swing2LimitDegrees=45, twistLimitDegrees=10}
 {bone1Name=dyn_poncho_front_ml_02, bone2Name=dyn_poncho_front_ml_01, swing1LimitDegrees=45, swing2LimitDegrees=45, twistLimitDegrees=10}
 {bone1Name=dyn_poncho_front_ml_03, bone2Name=dyn_poncho_front_ml_02, swing1LimitDegrees=45, swing2LimitDegrees=45, twistLimitDegrees=10}
 {bone1Name=dyn_poncho_front_ml_04, bone2Name=dyn_poncho_front_ml_03, swing1LimitDegrees=45, swing2LimitDegrees=45, twistLimitDegrees=10}
 {bone1Name=dyn_poncho_front_mid_01, bone2Name=dyn_poncho_front_mid_master, swing1LimitDegrees=45, swing2LimitDegrees=45, twistLimitDegrees=10}
 {bone1Name=dyn_poncho_front_mid_02, bone2Name=dyn_poncho_front_mid_01, swing1LimitDegrees=35, swing2LimitDegrees=35, twistLimitDegrees=0}
 {bone1Name=dyn_poncho_front_mid_03, bone2Name=dyn_poncho_front_mid_02, swing1LimitDegrees=45, swing2LimitDegrees=45, twistLimitDegrees=10}
 {bone1Name=dyn_poncho_front_mid_04, bone2Name=dyn_poncho_front_mid_03, swing1LimitDegrees=45, swing2LimitDegrees=45, twistLimitDegrees=10}
 {bone1Name=dyn_poncho_front_l_01, bone2Name=dyn_poncho_front_l_master, swing1LimitDegrees=45, swing2LimitDegrees=45, twistLimitDegrees=10}
 {bone1Name=dyn_poncho_front_l_02, bone2Name=dyn_poncho_front_l_01, swing1LimitDegrees=45, swing2LimitDegrees=45, twistLimitDegrees=10}
 {bone1Name=dyn_poncho_front_l_03, bone2Name=dyn_poncho_front_l_02, swing1LimitDegrees=45, swing2LimitDegrees=45, twistLimitDegrees=10}
 {bone1Name=dyn_poncho_front_l_04, bone2Name=dyn_poncho_front_l_03, swing1LimitDegrees=45, swing2LimitDegrees=45, twistLimitDegrees=10}
 {bone1Name=dyn_poncho_front_r_01, bone2Name=dyn_poncho_front_r_master, swing1LimitDegrees=45, swing2LimitDegrees=45, twistLimitDegrees=10}
 {bone1Name=dyn_poncho_front_r_02, bone2Name=dyn_poncho_front_r_01, swing1LimitDegrees=45, swing2LimitDegrees=45, twistLimitDegrees=10}
 {bone1Name=dyn_poncho_front_r_03, bone2Name=dyn_poncho_front_r_02, swing1LimitDegrees=45, swing2LimitDegrees=45, twistLimitDegrees=10}
 {bone1Name=dyn_poncho_front_r_04, bone2Name=dyn_poncho_front_r_03, swing1LimitDegrees=45, swing2LimitDegrees=45, twistLimitDegrees=10}
 {bone1Name=dyn_poncho_back_mid_01, bone2Name=spine_04, swing1LimitDegrees=45, swing2LimitDegrees=45, twistLimitDegrees=15}
 {bone1Name=dyn_poncho_back_mid_02, bone2Name=dyn_poncho_back_mid_01, swing1LimitDegrees=45, swing2LimitDegrees=45, twistLimitDegrees=15}
 {bone1Name=dyn_poncho_back_mid_03, bone2Name=dyn_poncho_back_mid_02, swing1LimitDegrees=45, swing2LimitDegrees=45, twistLimitDegrees=15}
 {bone1Name=dyn_poncho_back_mid_04, bone2Name=dyn_poncho_back_mid_03, swing1LimitDegrees=45, swing2LimitDegrees=45, twistLimitDegrees=15}
 {bone1Name=dyn_poncho_back_ml_01, bone2Name=spine_04, swing1LimitDegrees=45, swing2LimitDegrees=45, twistLimitDegrees=15}
 {bone1Name=dyn_poncho_back_ml_02, bone2Name=dyn_poncho_back_ml_01, swing1LimitDegrees=45, swing2LimitDegrees=45, twistLimitDegrees=15}
 {bone1Name=dyn_poncho_back_ml_03, bone2Name=dyn_poncho_back_ml_02, swing1LimitDegrees=45, swing2LimitDegrees=45, twistLimitDegrees=15}
 {bone1Name=dyn_poncho_back_ml_04, bone2Name=dyn_poncho_back_ml_03, swing1LimitDegrees=45, swing2LimitDegrees=45, twistLimitDegrees=15}
 {bone1Name=dyn_poncho_back_mr_01, bone2Name=spine_04, swing1LimitDegrees=45, swing2LimitDegrees=45, twistLimitDegrees=15}
 {bone1Name=dyn_poncho_back_mr_02, bone2Name=dyn_poncho_back_mr_01, swing1LimitDegrees=45, swing2LimitDegrees=45, twistLimitDegrees=15}
 {bone1Name=dyn_poncho_back_mr_03, bone2Name=dyn_poncho_back_mr_02, swing1LimitDegrees=45, swing2LimitDegrees=45, twistLimitDegrees=15}
 {bone1Name=dyn_poncho_back_mr_04, bone2Name=dyn_poncho_back_mr_03, swing1LimitDegrees=45, swing2LimitDegrees=45, twistLimitDegrees=15}
 {bone1Name=dyn_poncho_back_r_01, bone2Name=spine_04, swing1LimitDegrees=45, swing2LimitDegrees=45, twistLimitDegrees=15}
 {bone1Name=dyn_poncho_back_r_02, bone2Name=dyn_poncho_back_r_01, swing1LimitDegrees=45, swing2LimitDegrees=45, twistLimitDegrees=15}
 {bone1Name=dyn_poncho_back_r_03, bone2Name=dyn_poncho_back_r_02, swing1LimitDegrees=45, swing2LimitDegrees=45, twistLimitDegrees=15}
 {bone1Name=dyn_poncho_back_r_04, bone2Name=dyn_poncho_back_r_03, swing1LimitDegrees=45, swing2LimitDegrees=45, twistLimitDegrees=15}
 {bone1Name=dyn_poncho_back_l_01, bone2Name=spine_04, swing1LimitDegrees=45, swing2LimitDegrees=45, twistLimitDegrees=15}
 {bone1Name=dyn_poncho_back_l_02, bone2Name=dyn_poncho_back_l_01, swing1LimitDegrees=45, swing2LimitDegrees=45, twistLimitDegrees=15}
 {bone1Name=dyn_poncho_back_l_03, bone2Name=dyn_poncho_back_l_02, swing1LimitDegrees=45, swing2LimitDegrees=45, twistLimitDegrees=15}
 {bone1Name=dyn_poncho_back_l_04, bone2Name=dyn_poncho_back_l_03, swing1LimitDegrees=45, swing2LimitDegrees=45, twistLimitDegrees=15}
 {bone1Name=dyn_l_sleeve_mid_01, bone2Name=upperarm_l, swing1LimitDegrees=45, swing2LimitDegrees=45, twistLimitDegrees=20}
 {bone1Name=dyn_l_sleeve_back_01, bone2Name=upperarm_l, swing1LimitDegrees=45, swing2LimitDegrees=45, twistLimitDegrees=20}
 {bone1Name=dyn_l_sleeve_front_01, bone2Name=upperarm_l, swing1LimitDegrees=35, swing2LimitDegrees=35, twistLimitDegrees=20}
 {bone1Name=dyn_r_sleeve_mid_01, bone2Name=upperarm_r, swing1LimitDegrees=45, swing2LimitDegrees=45, twistLimitDegrees=20}
 {bone1Name=dyn_r_sleeve_back_01, bone2Name=upperarm_r, swing1LimitDegrees=45, swing2LimitDegrees=45, twistLimitDegrees=20}
 {bone1Name=dyn_r_sleeve_front_01, bone2Name=upperarm_r, swing1LimitDegrees=35, swing2LimitDegrees=35, twistLimitDegrees=20}
 {
  bone1Name=dyn_poncho_back_ml_02
  bone2Name=dyn_poncho_back_l_02
  swing1LimitDegrees=45
  swing1Motion=Free
  swing2LimitDegrees=45
  swing2Motion=Free
  twistLimitDegrees=45
  twistMotion=Free
 }
 {
  bone1Name=dyn_poncho_back_ml_02
  bone2Name=dyn_poncho_back_mid_02
  swing1LimitDegrees=45
  swing1Motion=Free
  swing2LimitDegrees=45
  swing2Motion=Free
  twistLimitDegrees=45
  twistMotion=Free
 }
 {
  bone1Name=dyn_poncho_back_mr_02
  bone2Name=dyn_poncho_back_mid_02
  swing1LimitDegrees=45
  swing1Motion=Free
  swing2LimitDegrees=45
  swing2Motion=Free
  twistLimitDegrees=45
  twistMotion=Free
 }
 {
  bone1Name=dyn_poncho_back_mr_02
  bone2Name=dyn_poncho_back_r_02
  swing1LimitDegrees=45
  swing1Motion=Free
  swing2LimitDegrees=45
  swing2Motion=Free
  twistLimitDegrees=45
  twistMotion=Free
 }
 {
  bone1Name=dyn_poncho_back_l_04
  bone2Name=dyn_poncho_back_ml_04
  swing1LimitDegrees=45
  swing1Motion=Free
  swing2LimitDegrees=45
  swing2Motion=Free
  twistLimitDegrees=45
  twistMotion=Free
 }
 {
  bone1Name=dyn_poncho_back_mid_04
  bone2Name=dyn_poncho_back_ml_04
  swing1LimitDegrees=45
  swing1Motion=Free
  swing2LimitDegrees=45
  swing2Motion=Free
  twistLimitDegrees=45
  twistMotion=Free
 }
 {
  bone1Name=dyn_poncho_back_mid_04
  bone2Name=dyn_poncho_back_mr_04
  swing1LimitDegrees=45
  swing1Motion=Free
  swing2LimitDegrees=45
  swing2Motion=Free
  twistLimitDegrees=45
  twistMotion=Free
 }
 {
  bone1Name=dyn_poncho_back_r_04
  bone2Name=dyn_poncho_back_mr_04
  swing1LimitDegrees=45
  swing1Motion=Free
  swing2LimitDegrees=45
  swing2Motion=Free
  twistLimitDegrees=45
  twistMotion=Free
 }
 {
  bone1Name=dyn_poncho_front_r_02
  bone2Name=dyn_poncho_front_mr_02
  swing1LimitDegrees=45
  swing1Motion=Free
  swing2LimitDegrees=45
  swing2Motion=Free
  twistLimitDegrees=45
  twistMotion=Free
 }
 {
  bone1Name=dyn_poncho_front_mid_02
  bone2Name=dyn_poncho_front_mr_02
  swing1LimitDegrees=45
  swing1Motion=Free
  swing2LimitDegrees=45
  swing2Motion=Free
  twistLimitDegrees=45
  twistMotion=Free
 }
 {
  bone1Name=dyn_poncho_front_l_02
  bone2Name=dyn_poncho_front_ml_02
  swing1LimitDegrees=45
  swing1Motion=Free
  swing2LimitDegrees=45
  swing2Motion=Free
  twistLimitDegrees=45
  twistMotion=Free
 }
 {
  bone1Name=dyn_poncho_front_mid_02
  bone2Name=dyn_poncho_front_ml_02
  swing1LimitDegrees=45
  swing1Motion=Free
  swing2LimitDegrees=45
  swing2Motion=Free
  twistLimitDegrees=45
  twistMotion=Free
 }
 {
  bone1Name=dyn_poncho_front_r_04
  bone2Name=dyn_poncho_front_mr_04
  swing1LimitDegrees=45
  swing1Motion=Free
  swing2LimitDegrees=45
  swing2Motion=Free
  twistLimitDegrees=45
  twistMotion=Free
 }
 {
  bone1Name=dyn_poncho_front_mid_04
  bone2Name=dyn_poncho_front_mr_04
  swing1LimitDegrees=45
  swing1Motion=Free
  swing2LimitDegrees=45
  swing2Motion=Free
  twistLimitDegrees=45
  twistMotion=Free
 }
 {
  bone1Name=dyn_poncho_front_l_04
  bone2Name=dyn_poncho_front_ml_04
  swing1LimitDegrees=45
  swing1Motion=Free
  swing2LimitDegrees=45
  swing2Motion=Free
  twistLimitDegrees=45
  twistMotion=Free
 }
 {
  bone1Name=dyn_poncho_front_mid_04
  bone2Name=dyn_poncho_front_ml_04
  swing1LimitDegrees=45
  swing1Motion=Free
  swing2LimitDegrees=45
  swing2Motion=Free
  twistLimitDegrees=45
  twistMotion=Free
 }
 {bone1Name=dyn_poncho_armpit_back_r, bone2Name=spine_02, swing1LimitDegrees=25, swing2LimitDegrees=25, twistLimitDegrees=20}
 {bone1Name=dyn_poncho_armpit_front_r, bone2Name=spine_02, swing1LimitDegrees=25, swing2LimitDegrees=25, twistLimitDegrees=20}
 {bone1Name=dyn_poncho_armpit_front_l, bone2Name=spine_02, swing1LimitDegrees=25, swing2LimitDegrees=25, twistLimitDegrees=20}
 {bone1Name=dyn_poncho_armpit_back_l, bone2Name=spine_02, swing1LimitDegrees=25, swing2LimitDegrees=25, twistLimitDegrees=20}
 {bone1Name=dyn_l_sleeve_back_01, bone2Name=dyn_l_sleeve_mid_01, swing1LimitDegrees=20, swing2LimitDegrees=20, twistLimitDegrees=20}
 {bone1Name=dyn_r_sleeve_back_01, bone2Name=dyn_r_sleeve_mid_01, swing1LimitDegrees=20, swing2LimitDegrees=20, twistLimitDegrees=20}
 {
  bone1Name=calf_r
  bone2Name=thigh_r
  swing1LimitDegrees=45
  swing1Motion=Free
  swing2LimitDegrees=45
  swing2Motion=Free
  twistLimitDegrees=45
  twistMotion=Free
 }
 {
  bone1Name=calf_l
  bone2Name=thigh_l
  swing1LimitDegrees=45
  swing1Motion=Free
  swing2LimitDegrees=45
  swing2Motion=Free
  twistLimitDegrees=45
  twistMotion=Free
 }
 {
  bone1Name=dyn_belt_knife_r_master
  bone2Name=pelvis
  swing1LimitDegrees=45
  swing1Motion=Free
  swing2LimitDegrees=45
  swing2Motion=Free
  twistLimitDegrees=45
  twistMotion=Free
 }
 {
  bone1Name=dyn_belt_knife_l_master
  bone2Name=pelvis
  swing1LimitDegrees=45
  swing1Motion=Free
  swing2LimitDegrees=45
  swing2Motion=Free
  twistLimitDegrees=45
  twistMotion=Free
 }
 {
  bone1Name=dyn_poncho_front_mr_master
  bone2Name=spine_04
  swing1LimitDegrees=45
  swing1Motion=Free
  swing2LimitDegrees=45
  swing2Motion=Free
  twistLimitDegrees=45
  twistMotion=Free
 }
 {
  bone1Name=dyn_poncho_front_ml_master
  bone2Name=spine_04
  swing1LimitDegrees=45
  swing1Motion=Free
  swing2LimitDegrees=45
  swing2Motion=Free
  twistLimitDegrees=45
  twistMotion=Free
 }
 {
  bone1Name=dyn_poncho_front_mid_master
  bone2Name=spine_04
  swing1LimitDegrees=45
  swing1Motion=Free
  swing2LimitDegrees=45
  swing2Motion=Free
  twistLimitDegrees=45
  twistMotion=Free
 }
 {
  bone1Name=dyn_poncho_front_r_master
  bone2Name=spine_04
  swing1LimitDegrees=45
  swing1Motion=Free
  swing2LimitDegrees=45
  swing2Motion=Free
  twistLimitDegrees=45
  twistMotion=Free
 }
 {
  bone1Name=dyn_poncho_front_l_master
  bone2Name=spine_04
  swing1LimitDegrees=45
  swing1Motion=Free
  swing2LimitDegrees=45
  swing2Motion=Free
  twistLimitDegrees=45
  twistMotion=Free
 }
]
