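# Physics asset for the Wombat Poncho medium male body's extras, covering secondary motion of dangling back straps, belt knives and knee strings.
# Capsules on pelvis, thighs and calves form the body base, linked by angularly free joints.
# Small capsule chains on dyn_ bones use tight swing and twist limits.
# Straps and knives hang from the pelvis, left knee strings from the left calf, right knee strings from the right thigh.
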PhysicsAsset /BRCosmetics/Characters/Player/Male/Medium/Bodies/M_MED_Wombat_Poncho/Meshes/M_MED_Wombat_Poncho_Physics_Extras
bodies {
 calf_l [{shapeType=Capsule, length=27, radius=6.5, center=<-19.0202 -2.11 0>, rotation=<rot 90 0 0>}]
 calf_r [{shapeType=Capsule, length=27, radius=6.5, center=<19 0.0005 0>, rotation=<rot 90 0 0>}]
 dyn_back_strap_01 [{
   shapeType=Capsule
   length=3.999
   radius=1.0089
   shapeName=dyn_back_strap_03_capsule
   center=<2.7791 0.0001 0>
   rotation=<rot 90 -26.565 -116.565>
  }]
 dyn_back_strap_02 [{
   shapeType=Capsule
   length=2.9391
   radius=2.0688
   shapeName=dyn_back_strap_03_capsule
   center=<2.7791 0.0001 0>
   rotation=<rot 90 -26.565 -116.565>
  }]
 dyn_back_strap_03 [{
   shapeType=Capsule
   length=2.9391
   radius=2.0688
   shapeName=dyn_back_strap_03_capsule
   center=<2.7791 0.0001 0>
   rotation=<rot 90 -26.565 -116.565>
  }]
 dyn_belt_knife_l_01 [{
   shapeType=Capsule
   length=3.6844
   radius=1.3234
   shapeName=dyn_back_strap_03_capsule
   center=<2.7791 0.0001 0>
   rotation=<rot 90 -26.565 -116.565>
  }]
 dyn_belt_knife_l_02 [{shapeType=Capsule, length=6.7082, radius=1.4598, shapeName=dyn_back_strap_03_capsule, center=<4.9624 0.0001 0.0001>, rotation=<rot 90 0 0>}]
 dyn_belt_knife_r_01 [{
   shapeType=Capsule
   length=3.6914
   radius=1.3164
   shapeName=dyn_back_strap_03_capsule
   center=<2.7791 0.0001 0>
   rotation=<rot 90 -26.565 -116.565>
  }]
 dyn_belt_knife_r_02 [{shapeType=Capsule, length=6.7082, radius=1.4598, shapeName=dyn_back_strap_03_capsule, center=<4.9624 0.0001 0.0001>, rotation=<rot 90 0 0>}]
 dyn_l_knee_loop_l_01 [{shapeType=Capsule, length=1.317, radius=0.5, shapeName=dyn_back_strap_03_capsule, center=<1.5166 0.0001 0>, rotation=<rot 90 0 0>}]
 dyn_l_knee_loop_l_02 [{shapeType=Capsule, length=1.4328, radius=0.5958, shapeName=dyn_back_strap_03_capsule, center=<1.5166 0.0001 0>, rotation=<rot 90 0 0>}]
 dyn_l_knee_string_long_01 [{shapeType=Capsule, length=3.923, radius=0.5207, shapeName=dyn_back_strap_03_capsule, center=<2.885 0.0001 -0.0001>, rotation=<rot 90 0 0>}]
 dyn_l_knee_string_long_02 [{shapeType=Capsule, length=4.1718, radius=0.5, shapeName=dyn_back_strap_03_capsule, center=<3.1149 0.0001 0>, rotation=<rot 90 0 0>}]
 dyn_l_knee_string_long_03 [{shapeType=Capsule, length=3.7206, radius=0.5013, shapeName=dyn_back_strap_03_capsule, center=<2.4434 0.0001 0>, rotation=<rot 90 0 0>}]
 dyn_l_knee_string_r_loop_01 [{shapeType=Capsule, length=1.1222, radius=0.6254, shapeName=dyn_back_strap_03_capsule, center=<1.1388 0.0001 0>, rotation=<rot 90 0 0>}]
 dyn_l_knee_string_r_loop_02 [{shapeType=Capsule, length=1.0704, radius=0.5, shapeName=dyn_back_strap_03_capsule, center=<1.3113 0.0001 0>, rotation=<rot 90 0 0>}]
 dyn_l_knee_string_r_loop_03 [{shapeType=Capsule, length=2.7278, radius=0.6254, shapeName=dyn_back_strap_03_capsule, center=<2.0392 0.0001 0>, rotation=<rot 90 0 0>}]
 dyn_r_knee_string_01 [{shapeType=Capsule, length=2.0662, radius=0.5746, shapeName=dyn_back_strap_03_capsule, center=<1.7134 0.0001 0>, rotation=<rot 90 0 0>}]
 dyn_r_knee_string_02 [{shapeType=Capsule, length=3.3751, radius=0.8468, shapeName=dyn_back_strap_03_capsule, center=<2.4434 0.0001 0>, rotation=<rot 90 0 0>}]
 dyn_r_knee_string_03 [{shapeType=Capsule, length=3.3751, radius=0.8468, shapeName=dyn_back_strap_03_capsule, center=<2.4434 0.0001 0>, rotation=<rot 90 0 0>}]
 dyn_r_knee_string_r_01 [{shapeType=Capsule, length=2.2094, radius=0.5, shapeName=dyn_back_strap_03_capsule, center=<1.9143 0.0001 0>, rotation=<rot 90 0 -93.5295>}]
 dyn_r_string_r_01 [{shapeType=Capsule, length=1.5432, radius=0.5, shapeName=dyn_back_strap_03_capsule, center=<1.5772 0.0001 0.0001>, rotation=<rot 90 0 0>}]
 dyn_r_string_r_02 [{shapeType=Capsule, length=1.5432, radius=0.5, shapeName=dyn_back_strap_03_capsule, center=<1.5772 0.0001 0.0001>, rotation=<rot 90 0 0>}]
 pelvis [{shapeType=Capsule, length=12.99, radius=10.952}, {shapeType=Capsule, length=11.241, radius=10.365, center=<-9.73 0 0>}]
 thigh_l [{shapeType=Capsule, length=26.5, radius=7.5, center=<-20 0.0005 0>, rotation=<rot 90 0 0>}]
 thigh_r [{shapeType=Capsule, length=29.396, radius=7.459, center=<22.4471 0 0>, rotation=<rot 90 0 0>}]
}
constraints [
 {
  bone1Name=calf_l
  bone2Name=thigh_l
  swing1LimitDegrees=45
  swing1Motion=Free
  swing2LimitDegrees=45
  swing2Motion=Free
  twistLimitDegrees=45
  twistMotion=Free
 }
 {
  bone1Name=calf_r
  bone2Name=thigh_r
  swing1LimitDegrees=45
  swing1Motion=Free
  swing2LimitDegrees=45
  swing2Motion=Free
  twistLimitDegrees=45
  twistMotion=Free
 }
 {bone1Name=dyn_belt_knife_l_02, bone2Name=dyn_belt_knife_l_01, swing1LimitDegrees=10, swing2LimitDegrees=10, twistLimitDegrees=5}
 {bone1Name=dyn_belt_knife_r_02, bone2Name=dyn_belt_knife_r_01, swing1LimitDegrees=7, swing2LimitDegrees=10, twistLimitDegrees=3}
 {bone1Name=dyn_l_knee_loop_l_01, bone2Name=calf_l, swing1LimitDegrees=10, swing2LimitDegrees=20, twistLimitDegrees=7}
 {bone1Name=dyn_l_knee_loop_l_02, bone2Name=dyn_l_knee_loop_l_01, swing1LimitDegrees=10, swing2LimitDegrees=20, twistLimitDegrees=7}
 {bone1Name=dyn_l_knee_string_r_loop_01, bone2Name=calf_l, swing1LimitDegrees=10, swing2LimitDegrees=10, twistLimitDegrees=5}
 {
  bone1Name=dyn_l_knee_string_r_loop_02
  bone2Name=dyn_l_knee_string_r_loop_01
  swing1LimitDegrees=10
  swing2LimitDegrees=10
  twistLimitDegrees=5
 }
 {
  bone1Name=dyn_l_knee_string_r_loop_03
  bone2Name=dyn_l_knee_string_r_loop_02
  swing1LimitDegrees=10
  swing2LimitDegrees=10
  twistLimitDegrees=5
 }
 {bone1Name=dyn_l_knee_string_long_01, bone2Name=calf_l, swing1LimitDegrees=10, swing2LimitDegrees=10, twistLimitDegrees=5}
 {bone1Name=dyn_l_knee_string_long_02, bone2Name=dyn_l_knee_string_long_01, swing1LimitDegrees=10, swing2LimitDegrees=10, twistLimitDegrees=5}
 {bone1Name=dyn_l_knee_string_long_03, bone2Name=dyn_l_knee_string_long_02, swing1LimitDegrees=10, swing2LimitDegrees=10, twistLimitDegrees=5}
 {bone1Name=dyn_r_knee_string_01, bone2Name=thigh_r, swing1LimitDegrees=7, swing2LimitDegrees=12, twistLimitDegrees=5}
 {bone1Name=dyn_r_knee_string_02, bone2Name=dyn_r_knee_string_01, swing1LimitDegrees=7, swing2LimitDegrees=12, twistLimitDegrees=5}
 {bone1Name=dyn_r_knee_string_03, bone2Name=dyn_r_knee_string_02, swing1LimitDegrees=7, swing2LimitDegrees=12, twistLimitDegrees=5}
 {bone1Name=dyn_r_knee_string_r_01, bone2Name=thigh_r, swing1LimitDegrees=10, swing2LimitDegrees=10, twistLimitDegrees=5}
 {bone1Name=dyn_r_string_r_01, bone2Name=thigh_r, swing1LimitDegrees=5, swing2LimitDegrees=10, twistLimitDegrees=5}
 {bone1Name=dyn_r_string_r_02, bone2Name=dyn_r_string_r_01, swing1LimitDegrees=5, swing2LimitDegrees=10, twistLimitDegrees=5}
 {
  bone1Name=thigh_l
  bone2Name=pelvis
  swing1LimitDegrees=45
  swing1Motion=Free
  swing2LimitDegrees=45
  swing2Motion=Free
  twistLimitDegrees=45
  twistMotion=Free
 }
 {
  bone1Name=thigh_r
  bone2Name=pelvis
  swing1LimitDegrees=45
  swing1Motion=Free
  swing2LimitDegrees=45
  swing2Motion=Free
  twistLimitDegrees=45
  twistMotion=Free
 }
 {bone1Name=dyn_belt_knife_r_01, bone2Name=pelvis, swing1LimitDegrees=7, swing2LimitDegrees=15, twistLimitDegrees=3}
 {bone1Name=dyn_belt_knife_l_01, bone2Name=pelvis, swing1LimitDegrees=7, swing2LimitDegrees=15, twistLimitDegrees=5}
 {bone1Name=dyn_back_strap_01, bone2Name=pelvis, swing1LimitDegrees=5, swing2LimitDegrees=10, twistLimitDegrees=5}
 {bone1Name=dyn_back_strap_02, bone2Name=dyn_back_strap_01, swing1LimitDegrees=15, swing2LimitDegrees=25, twistLimitDegrees=5}
 {bone1Name=dyn_back_strap_03, bone2Name=dyn_back_strap_02, swing1LimitDegrees=15, swing2LimitDegrees=25, twistLimitDegrees=5}
]
