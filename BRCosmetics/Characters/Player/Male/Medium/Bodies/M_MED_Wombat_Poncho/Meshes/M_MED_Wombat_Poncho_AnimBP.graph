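# Animation Blueprint for the M_MED_Wombat_Poncho body cosmetic, animating its poncho, sleeves, straps and dangling props over the copied parent-mesh pose.
# RigidBody nodes simulate cloth bones using separate physics assets for default, crouch and skydive states.
# Weighted constraints and additive bone offsets correct the result for aiming, sprinting, jumping, melee and front-end poses.
# Cached poses and bool blends chain these stages into a final Control Rig pass.
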AnimBlueprint /BRCosmetics/Characters/Player/Male/Medium/Bodies/M_MED_Wombat_Poncho/Meshes/M_MED_Wombat_Poncho_AnimBP
target_skeleton=&/M_MED_Wombat_Poncho_Skeleton
graphs [
 AnimGraph {nodes [
   n1 = SelectFloat(A=1, B=0, bPickA=Get bIsCrouching())
   n2 = "Transform (Modify) Bone - Bone: dyn_l_shoelaces_master"(
    ComponentPose=Local To Component(LocalPose=Use cached pose 'Final_Adjustments'())
    Alpha=@n1
    Translation="0,0,2"
    Rotation="0,0,0"
    Scale="1,1,1"
   ) {
    rotationMode=BMM_Additive
    scaleMode=BMM_Additive
    translationMode=BMM_Additive
    alphaBoolBlend {blendOption=Linear}
    boneToModify {boneName=dyn_l_shoelaces_master}
    translation {z=2}
   }
   n3 = "Transform (Modify) Bone - Bone: dyn_l_knee_string_long_master"(ComponentPose=@n2, Alpha=@n1, Translation="0,0,2", Rotation="0,0,0", Scale="1,1,1") {
    rotationMode=BMM_Additive
    scaleMode=BMM_Additive
    translationMode=BMM_Additive
    alphaBoolBlend {blendOption=Linear}
    boneToModify {boneName=dyn_l_knee_string_long_master}
    translation {z=2}
   }
   Blend Poses by bool(
    BlendPose_1=Blend Poses by bool(BlendPose_0=Component To Local(ComponentPose=@n3), BlendPose_1=Use cached pose 'Final_Adjustments'(), bActiveValue=Get bIsCrouching()) {blendType=Linear}
    bActiveValue=Get bIsInFrontEnd()
   ) {blendType=Linear}
   "Constraint - dyn_l_sleeve_back_01 (P)"(Alpha=1, ConstraintWeights_0=1) {
    alphaBoolBlend {blendOption=Linear}
    boneToModify {boneName=dyn_l_sleeve_back_01}
    constraintSetup [{targetBone {boneName=upperarm_l}}]
    constraintWeights [1]
   }
   n4 = "Constraint - dyn_l_shoelaces_master (PPP)"!(ComponentPose, Alpha=1, ConstraintWeights_0=0.7001, ConstraintWeights_1:float=0.1546, ConstraintWeights_2:float=0.1452) -> (Pose) {
    alphaBoolBlend {blendOption=Linear}
    boneToModify {boneName=dyn_l_shoelaces_master}
    constraintSetup [{targetBone {boneName=calf_l}}, {targetBone {boneName=calf_twist_02_l}}, {targetBone {boneName=thigh_l}}]
    constraintWeights [1,1,1]
   }
   "Constraint - dyn_l_knee_string_long_master (PPP)"!(ComponentPose=@n4, Alpha=1, ConstraintWeights_0=0.7001, ConstraintWeights_1:float=0.1546, ConstraintWeights_2:float=0.1452) -> (Pose) {
    alphaBoolBlend {blendOption=Linear}
    boneToModify {boneName=dyn_l_knee_string_long_master}
    constraintSetup [{targetBone {boneName=calf_l}}, {targetBone {boneName=calf_twist_02_l}}, {targetBone {boneName=thigh_l}}]
    constraintWeights [1,1,1]
   }
   "Constraint - dyn_r_sleeve_back_01 (P)"(Alpha=1, ConstraintWeights_0=1) {
    alphaBoolBlend {blendOption=Linear}
    boneToModify {boneName=dyn_r_sleeve_back_01}
    constraintSetup [{targetBone {boneName=upperarm_r}}]
    constraintWeights [1]
   }
   Local To Component(LocalPose=Blend Poses by bool(BlendPose_0=Component To Local(), BlendPose_1=Copy Pose From Mesh() {bUseAttachedParent=true}, bActiveValue=Get bIsInFrontEnd()) {blendType=Linear})
   Local To Component(LocalPose=Blend Poses by bool(BlendPose_0=Component To Local(), BlendPose_1=Use cached pose 'Final_Adjustments'()) {blendType=Linear})
   Local To Component(LocalPose=Use cached pose 'Final_Adjustments'())
   n5 = Get Aim_Up() -> (Aim_Up:double)
   "Transform (Modify) Bone - Bone: dyn_poncho_back_l_01"(Alpha=@n5, Translation="4,4,1.5", Rotation="0,0,0", Scale="1,1,1") {
    rotationMode=BMM_Additive
    scaleMode=BMM_Additive
    translationMode=BMM_Additive
    alphaBoolBlend {blendOption=Linear}
    boneToModify {boneName=dyn_poncho_back_l_01}
    translation=<4 4 1.5>
   }
   n6 = Get is_sprinting_float() -> (is_sprinting_float:double)
   "Transform (Modify) Bone - Bone: dyn_poncho_armpit_back_r"(Alpha=@n6, Translation="0,-15,0", Rotation="0,0,0", Scale="1,1,1") {
    rotationMode=BMM_Additive
    scaleMode=BMM_Additive
    translationMode=BMM_Additive
    alphaBoolBlend {blendOption=Linear}
    boneToModify {boneName=dyn_poncho_armpit_back_r}
    translation {y=-15}
   }
   n7 = "Transform (Modify) Bone - Bone: dyn_r_sleeve_back_master"(Alpha=@n6, Translation="0,0,-5", Rotation="0,0,0", Scale="1,1,1") {
    rotationMode=BMM_Additive
    rotationSpace=BCS_BoneSpace
    scaleMode=BMM_Additive
    translationMode=BMM_Additive
    translationSpace=BCS_BoneSpace
    alphaBoolBlend {blendOption=Linear}
    boneToModify {boneName=dyn_r_sleeve_back_master}
    translation {z=-5}
   }
   "Transform (Modify) Bone - Bone: dyn_poncho_back_r_master"(ComponentPose=@n7, Alpha=@n6, Translation="0,5,-3", Rotation="0,0,0", Scale="1,1,1") {
    rotationMode=BMM_Additive
    rotationSpace=BCS_BoneSpace
    scaleMode=BMM_Additive
    translationMode=BMM_Additive
    translationSpace=BCS_BoneSpace
    alphaBoolBlend {blendOption=Linear}
    boneToModify {boneName=dyn_poncho_back_r_master}
    translation {y=5, z=-3}
   }
   "Transform (Modify) Bone - Bone: dyn_belt_knife_l_02"(Alpha=1, Translation="0,0,0", Rotation="-12,0,0", Scale="1,1,1") {
    rotationMode=BMM_Additive
    translationMode=BMM_Additive
    alphaBoolBlend {blendOption=Linear}
    boneToModify {boneName=dyn_belt_knife_l_02}
    rotation {pitch=-12}
   }
   RigidBody() {
    bClampLinearTranslationLimitToRefPose=true
    bOverrideWorldGravity=true
    overridePhysicsAsset=&/M_MED_Wombat_Poncho_Physics_Extras
    simulationSpace=BaseBoneSpace
    alphaBoolBlend {blendOption=Linear}
    baseBoneRef {boneName=root}
    componentAppliedLinearAccClamp=<1200 1200 1200>
    componentLinearAccScale=<0.35 0.35 0.45>
    overrideWorldGravity {z=-980}
   }
   n8 = "Constraint - dyn_l_sleeve_back_01 (P)"(ComponentPose=Local To Component(LocalPose=Use cached pose 'Pick_Adjust'()), Alpha=1, ConstraintWeights_0=1) {
    alphaBoolBlend {blendOption=Linear}
    boneToModify {boneName=dyn_l_sleeve_back_01}
    constraintSetup [{targetBone {boneName=upperarm_l}}]
    constraintWeights [1]
   }
   n9 = "Constraint - dyn_poncho_armpit_back_l (P)"(ComponentPose=@n8, Alpha=1, ConstraintWeights_0=1) {
    alphaBoolBlend {blendOption=Linear}
    boneToModify {boneName=dyn_poncho_armpit_back_l}
    constraintSetup [{targetBone {boneName=upperarm_l}}]
    constraintWeights [1]
   }
   n10 = "Transform (Modify) Bone - Bone: dyn_poncho_armpit_back_l"(ComponentPose=@n9, Alpha=1, Translation="5,-8,5", Rotation="0,0,0", Scale="1,1,1") {
    rotationMode=BMM_Additive
    scaleMode=BMM_Additive
    translationMode=BMM_Additive
    alphaBoolBlend {blendOption=Linear}
    boneToModify {boneName=dyn_poncho_armpit_back_l}
    translation=<5 -8 5>
   }
   n11 = "Constraint - dyn_r_sleeve_back_01 (P)"(ComponentPose=@n10, Alpha=1, ConstraintWeights_0=1) {
    alphaBoolBlend {blendOption=Linear}
    boneToModify {boneName=dyn_r_sleeve_back_01}
    constraintSetup [{targetBone {boneName=upperarm_r}}]
    constraintWeights [1]
   }
   n12 = "Constraint - dyn_poncho_armpit_back_r (PP)"!(ComponentPose=@n11, Alpha=1, ConstraintWeights_0=1, ConstraintWeights_1:float=0) -> (Pose) {
    alphaBoolBlend {blendOption=Linear}
    boneToModify {boneName=dyn_poncho_armpit_back_r}
    constraintSetup [{targetBone {boneName=upperarm_r}}, {targetBone {boneName=dyn_poncho_back_r_01}}]
    constraintWeights [1,1]
   }
   n13 = "Transform (Modify) Bone - Bone: dyn_poncho_armpit_back_r"(ComponentPose=@n12, Alpha=1, Translation="-5,-8,5", Rotation="0,0,0", Scale="1,1,1") {
    rotationMode=BMM_Additive
    scaleMode=BMM_Additive
    translationMode=BMM_Additive
    alphaBoolBlend {blendOption=Linear}
    boneToModify {boneName=dyn_poncho_armpit_back_r}
    translation=<-5 -8 5>
   }
   n14 = "Transform (Modify) Bone - Bone: dyn_poncho_back_l_01"(ComponentPose=@n13, Alpha=1, Translation="0,-5,5", Rotation="0,0,0", Scale="1,1,1") {
    rotationMode=BMM_Additive
    scaleMode=BMM_Additive
    translationMode=BMM_Additive
    alphaBoolBlend {blendOption=Linear}
    boneToModify {boneName=dyn_poncho_back_l_01}
    translation {y=-5, z=5}
   }
   n15 = "Transform (Modify) Bone - Bone: dyn_poncho_back_r_01"(ComponentPose=@n14, Alpha=1, Translation="0,-5,5", Rotation="0,0,0", Scale="1,1,1") {
    rotationMode=BMM_Additive
    scaleMode=BMM_Additive
    translationMode=BMM_Additive
    alphaBoolBlend {blendOption=Linear}
    boneToModify {boneName=dyn_poncho_back_r_01}
    translation {y=-5, z=5}
   }
   n16 = Component To Local(ComponentPose=@n15)
   n17 = Use cached pose 'Pick_Adjust'()
   n18 = Get bIsPlayingMeleeAnim()
   n19 = Get Crouch_Float() -> (Crouch_Float:double)
   n20 = "Transform (Modify) Bone - Bone: dyn_l_shoelaces_master"(
    ComponentPose=Local To Component(LocalPose=Blend Poses by bool(BlendPose_0=@n16, BlendPose_1=@n17, BlendTime_1=0.3, bActiveValue=@n18) {blendType=Linear})
    Alpha=@n19
    Translation="0,0,2"
    Rotation="0,0,0"
    Scale="1,1,1"
   ) {
    rotationMode=BMM_Additive
    scaleMode=BMM_Additive
    translationMode=BMM_Additive
    alphaBoolBlend {blendOption=Linear}
    boneToModify {boneName=dyn_l_shoelaces_master}
    translation {z=2}
   }
   n21 = "Transform (Modify) Bone - Bone: dyn_l_knee_string_long_master"(ComponentPose=@n20, Alpha=@n19, Translation="0,0,2", Rotation="0,0,0", Scale="1,1,1") {
    rotationMode=BMM_Additive
    scaleMode=BMM_Additive
    translationMode=BMM_Additive
    alphaBoolBlend {blendOption=Linear}
    boneToModify {boneName=dyn_l_knee_string_long_master}
    translation {z=2}
   }
   n22 = "Transform (Modify) Bone - Bone: dyn_poncho_armpit_back_r"(ComponentPose=@n21, Alpha=Get melee_pickaxe() -> (melee_pickaxe:double), Translation="5,-3,0", Rotation="0,0,0", Scale="1,1,1") {
    rotationMode=BMM_Additive
    translationMode=BMM_Additive
    alphaBoolBlend {blendOption=Linear}
    boneToModify {boneName=dyn_poncho_armpit_back_r}
    translation=<5 -3>
   }
   n23 = "Transform (Modify) Bone - Bone: dyn_poncho_back_mr_master"!(ComponentPose=@n22, Alpha=1, Translation="0,-2,0", Rotation="0.000,0.000,10.000") -> (Pose) {
    rotationMode=BMM_Additive
    translationMode=BMM_Additive
    alphaBoolBlend {blendOption=Linear}
    alphaScaleBiasClamp {bInterpResult=true, interpSpeedDecreasing=5, interpSpeedIncreasing=7}
    boneToModify {boneName=dyn_poncho_back_mr_master}
    rotation {roll=10}
    translation {y=-2}
   }
   n24 = "Transform (Modify) Bone - Bone: dyn_poncho_back_mid_master"!(ComponentPose=@n23, Alpha=1, Translation="0,-2,0", Rotation="0,0,10") -> (Pose) {
    rotationMode=BMM_Additive
    translationMode=BMM_Additive
    alphaBoolBlend {blendOption=Linear}
    alphaScaleBiasClamp {bInterpResult=true, interpSpeedDecreasing=5, interpSpeedIncreasing=7}
    boneToModify {boneName=dyn_poncho_back_mid_master}
    rotation {roll=10}
    translation {y=-2}
   }
   n25 = "Transform (Modify) Bone - Bone: dyn_belt_knife_l_02"(ComponentPose=@n24, Alpha=1, Translation="0,-2,0", Rotation="0,0,0", Scale="1,1,1") {
    rotationMode=BMM_Additive
    translationMode=BMM_Additive
    alphaBoolBlend {blendOption=Linear}
    boneToModify {boneName=dyn_belt_knife_l_02}
    translation {y=-2}
   }
   n26 = "Transform (Modify) Bone - Bone: dyn_poncho_armpit_back_r"(ComponentPose=@n25, Alpha=1, Translation="4,0,0", Rotation="0,0,0", Scale="1,1,1") {
    rotationMode=BMM_Additive
    scaleMode=BMM_Additive
    translationMode=BMM_Additive
    alphaBoolBlend {blendOption=Linear}
    boneToModify {boneName=dyn_poncho_armpit_back_r}
    translation {x=4}
   }
   n27 = Control Rig(Source=Component To Local(ComponentPose=@n26)) {
    alphaBoolBlend {blendOption=Linear}
    controlRigAssetReference {blueprintRigClass=&/M_MED_Wombat_Poncho_CtrlRig.M_MED_Wombat_Poncho_CtrlRig_C}
   }
   Output Pose(Result=@n27) {layerGroup=None}
   n28 = "Transform (Modify) Bone - Bone: dyn_poncho_armpit_back_r"(ComponentPose=Local To Component(LocalPose=Use cached pose 'Pre_Melee'()), Alpha=1, Translation="0,-5,0", Rotation="0,0,0", Scale="1,1,1") {
    rotationMode=BMM_Additive
    translationMode=BMM_Additive
    translationSpace=BCS_BoneSpace
    alphaBoolBlend {blendOption=Linear}
    boneToModify {boneName=dyn_poncho_armpit_back_r}
    translation {y=-5}
   }
   n29 = "Transform (Modify) Bone - Bone: dyn_r_sleeve_back_master"(ComponentPose=@n28, Alpha=1, Translation="0,-3,0", Rotation="0,0,0", Scale="1,1,1") {
    rotationMode=BMM_Additive
    translationMode=BMM_Additive
    alphaBoolBlend {blendOption=Linear}
    boneToModify {boneName=dyn_r_sleeve_back_master}
    translation {y=-3}
   }
   n30 = "Transform (Modify) Bone - Bone: dyn_poncho_back_ml_master"(ComponentPose=@n29, Alpha=1, Translation="0,-5,0", Rotation="0,0,5", Scale="1,1,1") {
    rotationMode=BMM_Additive
    translationMode=BMM_Additive
    translationSpace=BCS_BoneSpace
    alphaBoolBlend {blendOption=Linear}
    boneToModify {boneName=dyn_poncho_back_ml_master}
    rotation {roll=5}
    translation {y=-5}
   }
   n31 = "Transform (Modify) Bone - Bone: dyn_poncho_back_mid_master"!(ComponentPose=@n30, Alpha=1, Translation="0,-5,0", Rotation="0,0,5", Scale:Vector="1,1,1") -> (Pose) {
    rotationMode=BMM_Additive
    translationMode=BMM_Additive
    alphaBoolBlend {blendOption=Linear}
    boneToModify {boneName=dyn_poncho_back_mid_master}
    rotation {roll=5}
    translation {y=-5}
   }
   n32 = "Transform (Modify) Bone - Bone: dyn_poncho_back_mr_master"!(ComponentPose=@n31, Alpha=1, Translation="0,-2,0", Rotation="0,0,5", Scale:Vector="1,1,1") -> (Pose) {
    rotationMode=BMM_Additive
    translationMode=BMM_Additive
    alphaBoolBlend {blendOption=Linear}
    boneToModify {boneName=dyn_poncho_back_mr_master}
    rotation {roll=5}
    translation {y=-2}
   }
   n33 = Blend Poses by bool(
    BlendPose_0=Component To Local(ComponentPose=@n32)
    BlendPose_1=Use cached pose 'Pre_Melee'()
    BlendTime_0=0.3
    BlendTime_1=0.6
    bActiveValue=Get JumpFall() -> (JumpFall:bool)
   ) {blendType=Linear}
   n34 = "Transform (Modify) Bone - Bone: dyn_r_sleeve_back_master"(ComponentPose=Local To Component(LocalPose=Use cached pose 'Pre_Melee'()), Alpha=1, Translation="0,0,-5", Rotation="0,0,0", Scale="1,1,1") {
    rotationMode=BMM_Additive
    rotationSpace=BCS_BoneSpace
    scaleMode=BMM_Additive
    translationMode=BMM_Additive
    translationSpace=BCS_BoneSpace
    alphaBoolBlend {blendOption=Linear}
    boneToModify {boneName=dyn_r_sleeve_back_master}
    translation {z=-5}
   }
   n35 = "Transform (Modify) Bone - Bone: dyn_poncho_back_r_master"(ComponentPose=@n34, Alpha=1, Translation="0,5,-3", Rotation="0,0,0", Scale="1,1,1") {
    rotationMode=BMM_Additive
    rotationSpace=BCS_BoneSpace
    scaleMode=BMM_Additive
    translationMode=BMM_Additive
    translationSpace=BCS_BoneSpace
    alphaBoolBlend {blendOption=Linear}
    boneToModify {boneName=dyn_poncho_back_r_master}
    translation {y=5, z=-3}
   }
   n36 = "Transform (Modify) Bone - Bone: dyn_poncho_armpit_back_r"(ComponentPose=@n35, Alpha=1, Translation="0,-15,0", Rotation="0,0,0", Scale="1,1,1") {
    rotationMode=BMM_Additive
    scaleMode=BMM_Additive
    translationMode=BMM_Additive
    alphaBoolBlend {blendOption=Linear}
    boneToModify {boneName=dyn_poncho_armpit_back_r}
    translation {y=-15}
   }
   n37 = Component To Local(ComponentPose=@n36)
   n38 = Get bIsSprinting()
   n39 = "Constraint - dyn_poncho_front_r_01 (P)"(ComponentPose=Local To Component(LocalPose=Copy Pose From Mesh() {bUseAttachedParent=true}), Alpha=0.9, ConstraintWeights_0=1) {
    alphaBoolBlend {blendOption=Linear}
    boneToModify {boneName=dyn_poncho_front_r_01}
    constraintSetup [{targetBone {boneName=upperarm_r}}]
    constraintWeights [1]
   }
   n40 = RigidBody!(ComponentPose=@n39, Alpha, OverrideWorldGravity:Vector=Get GravityOverride(), ExternalForce) -> (Pose) {
    bClampLinearTranslationLimitToRefPose=true
    bOverrideWorldGravity=true
    overridePhysicsAsset=&/M_MED_Wombat_Poncho_Skydive
    simulationSpace=BaseBoneSpace
    alphaBoolBlend {blendOption=Linear}
    baseBoneRef {boneName=spine_02}
    componentAppliedLinearAccClamp=<3000 3000 3000>
    componentLinearAccScale=<0.75 0.25 0.2>
    componentLinearVelScale=<0.35 0.2 0.2>
    overrideWorldGravity {z=-980}
   }
   n41 = RigidBody(ComponentPose=@n40, OverrideWorldGravity="0,0,-270") {
    bClampLinearTranslationLimitToRefPose=true
    bOverrideWorldGravity=true
    overridePhysicsAsset=&/M_MED_Wombat_Poncho_FRont_Skydive
    simulationSpace=BaseBoneSpace
    alphaBoolBlend {blendOption=Linear}
    baseBoneRef {boneName=spine_02}
    componentAppliedLinearAccClamp=<3000 3000 3000>
    componentLinearAccScale=<0.75 0.25 0.2>
    componentLinearVelScale=<0.35 0.2 0.2>
    overrideWorldGravity {z=-980}
   }
   n42 = RigidBody!(ComponentPose=@n41, Alpha, ExternalForce) -> (Pose) {
    bClampLinearTranslationLimitToRefPose=true
    bOverrideWorldGravity=true
    overridePhysicsAsset=&M_MED_Wombat_Poncho_Physics_Extras
    simulationSpace=BaseBoneSpace
    alphaBoolBlend {blendOption=Linear}
    baseBoneRef {boneName=root}
    componentAppliedLinearAccClamp=<1200 1200 1200>
    componentLinearAccScale=<0.35 0.35 0.45>
    overrideWorldGravity {z=-980}
   }
   n43 = "Constraint - dyn_back_strap_master (PPP)"!(ComponentPose=@n42, Alpha=1, ConstraintWeights_0=0.0331, ConstraintWeights_1:float=0.157, ConstraintWeights_2:float=0.8099) -> (Pose) {
    alphaBoolBlend {blendOption=Linear}
    boneToModify {boneName=dyn_back_strap_master}
    constraintSetup [{targetBone {boneName=spine_01}}, {targetBone {boneName=thigh_twist_01_r}}, {targetBone {boneName=pelvis}}]
    constraintWeights [1,1,1]
   }
   n44 = "Constraint - dyn_l_sleeve_mid_01 (P)"(ComponentPose=@n43, Alpha=1, ConstraintWeights_0=1) {
    alphaBoolBlend {blendOption=Linear}
    boneToModify {boneName=dyn_l_sleeve_mid_01}
    constraintSetup [{targetBone {boneName=upperarm_l}}]
    constraintWeights [1]
   }
   n45 = "Constraint - dyn_l_sleeve_back_01 (P)"(ComponentPose=@n44, Alpha=1, ConstraintWeights_0=1) {
    alphaBoolBlend {blendOption=Linear}
    boneToModify {boneName=dyn_l_sleeve_back_01}
    constraintSetup [{targetBone {boneName=upperarm_l}}]
    constraintWeights [1]
   }
   n46 = "Constraint - dyn_l_sleeve_front_01 (P)"(ComponentPose=@n45, Alpha=1, ConstraintWeights_0=1) {
    alphaBoolBlend {blendOption=Linear}
    boneToModify {boneName=dyn_l_sleeve_front_01}
    constraintSetup [{targetBone {boneName=upperarm_l}}]
    constraintWeights [1]
   }
   n47 = "Constraint - dyn_r_sleeve_mid_01 (P)"(ComponentPose=@n46, Alpha=1, ConstraintWeights_0=1) {
    alphaBoolBlend {blendOption=Linear}
    boneToModify {boneName=dyn_r_sleeve_mid_01}
    constraintSetup [{targetBone {boneName=upperarm_r}}]
    constraintWeights [1]
   }
   n48 = "Constraint - dyn_r_sleeve_back_01 (P)"(ComponentPose=@n47, Alpha=1, ConstraintWeights_0=1) {
    alphaBoolBlend {blendOption=Linear}
    boneToModify {boneName=dyn_r_sleeve_back_01}
    constraintSetup [{targetBone {boneName=upperarm_r}}]
    constraintWeights [1]
   }
   n49 = "Constraint - dyn_r_sleeve_front_01 (P)"(ComponentPose=@n48, Alpha=1, ConstraintWeights_0=1) {
    alphaBoolBlend {blendOption=Linear}
    boneToModify {boneName=dyn_r_sleeve_front_01}
    constraintSetup [{targetBone {boneName=upperarm_r}}]
    constraintWeights [1]
   }
   n50 = "Constraint - dyn_poncho_armpit_back_l (PP)"!(ComponentPose=@n49, Alpha=1, ConstraintWeights_0=1, ConstraintWeights_1:float=0) -> (Pose) {
    alphaBoolBlend {blendOption=Linear}
    boneToModify {boneName=dyn_poncho_armpit_back_l}
    constraintSetup [{targetBone {boneName=upperarm_l}}, {targetBone {boneName=dyn_poncho_back_l_01}}]
    constraintWeights [1,1]
   }
   n51 = "Constraint - dyn_poncho_armpit_front_l (PP)"!(ComponentPose=@n50, Alpha=1, ConstraintWeights_0=1, ConstraintWeights_1:float=0) -> (Pose) {
    alphaBoolBlend {blendOption=Linear}
    boneToModify {boneName=dyn_poncho_armpit_front_l}
    constraintSetup [{targetBone {boneName=upperarm_l}}, {targetBone {boneName=dyn_poncho_front_l_01}}]
    constraintWeights [1,1]
   }
   n52 = "Constraint - dyn_poncho_armpit_front_r (PP)"!(ComponentPose=@n51, Alpha=1, ConstraintWeights_0=1, ConstraintWeights_1:float=0) -> (Pose) {
    alphaBoolBlend {blendOption=Linear}
    boneToModify {boneName=dyn_poncho_armpit_front_r}
    constraintSetup [{targetBone {boneName=upperarm_r}}, {targetBone {boneName=dyn_poncho_front_r_01}}]
    constraintWeights [1,1]
   }
   n53 = "Constraint - dyn_poncho_armpit_back_r (PP)"(ComponentPose=@n52, Alpha=1, ConstraintWeights_0=1, ConstraintWeights_1=0) {
    alphaBoolBlend {blendOption=Linear}
    boneToModify {boneName=dyn_poncho_armpit_back_r}
    constraintSetup [{targetBone {boneName=upperarm_r}}, {targetBone {boneName=dyn_poncho_back_r_01}}]
    constraintWeights [1,1]
   }
   n54 = "Constraint - dyn_poncho_back_l_01 (P)"(ComponentPose=@n53, Alpha=0.2, ConstraintWeights_0=1) {
    alphaBoolBlend {blendOption=Linear}
    boneToModify {boneName=dyn_poncho_back_l_01}
    constraintSetup [{targetBone {boneName=upperarm_l}}]
    constraintWeights [1]
   }
   n55 = "Constraint - dyn_poncho_back_r_01 (P)"(ComponentPose=@n54, Alpha=0.2, ConstraintWeights_0=1) {
    alphaBoolBlend {blendOption=Linear}
    boneToModify {boneName=dyn_poncho_back_r_01}
    constraintSetup [{targetBone {boneName=upperarm_r}}]
    constraintWeights [1]
   }
   n56 = "Transform (Modify) Bone - Bone: dyn_poncho_back_l_01"(ComponentPose=@n55, Alpha=1, Translation="0,-5,10", Rotation="0,0,0", Scale="1,1,1") {
    rotationMode=BMM_Additive
    scaleMode=BMM_Additive
    translationMode=BMM_Additive
    alphaBoolBlend {blendOption=Linear}
    boneToModify {boneName=dyn_poncho_back_l_01}
    translation {y=-5, z=10}
   }
   n57 = "Transform (Modify) Bone - Bone: dyn_poncho_back_r_01"(ComponentPose=@n56, Alpha=1, Translation="0,-5,10", Rotation="0,0,0", Scale="1,1,1") {
    rotationMode=BMM_Additive
    scaleMode=BMM_Additive
    translationMode=BMM_Additive
    alphaBoolBlend {blendOption=Linear}
    boneToModify {boneName=dyn_poncho_back_r_01}
    translation {y=-5, z=10}
   }
   n58 = "Transform (Modify) Bone - Bone: dyn_l_sleeve_back_01"(ComponentPose=@n57, Alpha=1, Translation="0,-5,5", Rotation="0,0,0", Scale="1,1,1") {
    rotationMode=BMM_Additive
    scaleMode=BMM_Additive
    translationMode=BMM_Additive
    alphaBoolBlend {blendOption=Linear}
    boneToModify {boneName=dyn_l_sleeve_back_01}
    translation {y=-5, z=5}
   }
   n59 = "Transform (Modify) Bone - Bone: dyn_r_sleeve_back_01"(ComponentPose=@n58, Alpha=1, Translation="0,-5,5", Rotation="0,0,0", Scale="1,1,1") {
    rotationMode=BMM_Additive
    scaleMode=BMM_Additive
    translationMode=BMM_Additive
    alphaBoolBlend {blendOption=Linear}
    boneToModify {boneName=dyn_r_sleeve_back_01}
    translation {y=-5, z=5}
   }
   n60 = "Transform (Modify) Bone - Bone: dyn_poncho_back_mid_master"(ComponentPose=@n59, Alpha=1, Translation="0,0,-3", Rotation="0,0,0", Scale="1,1,1") {
    rotationMode=BMM_Additive
    scaleMode=BMM_Additive
    translationMode=BMM_Additive
    translationSpace=BCS_BoneSpace
    alphaBoolBlend {blendOption=Linear}
    boneToModify {boneName=dyn_poncho_back_mid_master}
    translation {z=-3}
   }
   n61 = "Transform (Modify) Bone - Bone: dyn_poncho_back_l_master"(ComponentPose=@n60, Alpha=1, Translation="0,0,-3", Rotation="0,0,0", Scale="1,1,1") {
    rotationMode=BMM_Additive
    scaleMode=BMM_Additive
    translationMode=BMM_Additive
    translationSpace=BCS_BoneSpace
    alphaBoolBlend {blendOption=Linear}
    boneToModify {boneName=dyn_poncho_back_l_master}
    translation {z=-3}
   }
   n62 = "Transform (Modify) Bone - Bone: dyn_poncho_back_r_master"(ComponentPose=@n61, Alpha=1, Translation="0,0,-3", Rotation="0,0,0", Scale="1,1,1") {
    rotationMode=BMM_Additive
    scaleMode=BMM_Additive
    translationMode=BMM_Additive
    translationSpace=BCS_BoneSpace
    alphaBoolBlend {blendOption=Linear}
    boneToModify {boneName=dyn_poncho_back_r_master}
    translation {z=-3}
   }
   n63 = "Transform (Modify) Bone - Bone: dyn_poncho_back_mr_master"(ComponentPose=@n62, Alpha=1, Translation="0,0,-3", Rotation="0,0,0", Scale="1,1,1") {
    rotationMode=BMM_Additive
    scaleMode=BMM_Additive
    translationMode=BMM_Additive
    translationSpace=BCS_BoneSpace
    alphaBoolBlend {blendOption=Linear}
    boneToModify {boneName=dyn_poncho_back_mr_master}
    translation {z=-3}
   }
   n64 = "Transform (Modify) Bone - Bone: dyn_poncho_back_ml_master"(ComponentPose=@n63, Alpha=1, Translation="0,0,-3", Rotation="0,0,0", Scale="1,1,1") {
    rotationMode=BMM_Additive
    scaleMode=BMM_Additive
    translationMode=BMM_Additive
    translationSpace=BCS_BoneSpace
    alphaBoolBlend {blendOption=Linear}
    boneToModify {boneName=dyn_poncho_back_ml_master}
    translation {z=-3}
   }
   Save cached pose 'Pick_Adjust'(Pose=Blend Poses by bool(
     BlendPose_0=Component To Local(ComponentPose=@n64)
     BlendPose_1=Blend Poses by bool(BlendPose_0=@n37, BlendPose_1=@n33, BlendTime_0=0.3, BlendTime_1=0.3, bActiveValue=@n38) {blendType=Linear}
     BlendTime_1=0.5
     bActiveValue=Get Skydive() -> (Skydive:bool)
    ) {blendType=Linear}) {cacheName=Pick_Adjust}
   n65 = Get Frontend_Kick_Out() -> (Frontend_Kick_Out:double)
   n66 = "Transform (Modify) Bone - Bone: dyn_poncho_front_r_01"(
    ComponentPose=Local To Component(LocalPose=Copy Pose From Mesh() {bUseAttachedParent=true})
    Alpha=@n65
    Translation="0,0,0"
    Rotation="0,0,-30"
    Scale="1,1,1"
   ) {
    rotationMode=BMM_Additive
    translationMode=BMM_Additive
    alphaBoolBlend {blendOption=Linear}
    boneToModify {boneName=dyn_poncho_front_r_01}
    rotation {roll=-30}
   }
   n67 = "Transform (Modify) Bone - Bone: dyn_poncho_front_mr_01"(ComponentPose=@n66, Alpha=@n65, Translation="0,0,0", Rotation="0,0,-30", Scale="1,1,1") {
    rotationMode=BMM_Additive
    translationMode=BMM_Additive
    alphaBoolBlend {blendOption=Linear}
    boneToModify {boneName=dyn_poncho_front_mr_01}
    rotation {roll=-30}
   }
   n68 = "Transform (Modify) Bone - Bone: dyn_poncho_front_mid_01"(ComponentPose=@n67, Alpha=@n65, Translation="0,0,0", Rotation="0,0,-30", Scale="1,1,1") {
    rotationMode=BMM_Additive
    translationMode=BMM_Additive
    alphaBoolBlend {blendOption=Linear}
    boneToModify {boneName=dyn_poncho_front_mid_01}
    rotation {roll=-30}
   }
   n69 = "Transform (Modify) Bone - Bone: dyn_poncho_front_ml_01"(ComponentPose=@n68, Alpha=@n65, Translation="0,0,0", Rotation="0,0,-30", Scale="1,1,1") {
    rotationMode=BMM_Additive
    translationMode=BMM_Additive
    alphaBoolBlend {blendOption=Linear}
    boneToModify {boneName=dyn_poncho_front_ml_01}
    rotation {roll=-30}
   }
   n70 = "Transform (Modify) Bone - Bone: dyn_poncho_front_l_01"(ComponentPose=@n69, Alpha=@n65, Translation="0,0,0", Rotation="0,0,-30", Scale="1,1,1") {
    rotationMode=BMM_Additive
    translationMode=BMM_Additive
    alphaBoolBlend {blendOption=Linear}
    boneToModify {boneName=dyn_poncho_front_l_01}
    rotation {roll=-30}
   }
   n71 = "Constraint - dyn_poncho_front_r_01 (P)"(ComponentPose=@n70, Alpha=0.9, ConstraintWeights_0=1) {
    alphaBoolBlend {blendOption=Linear}
    boneToModify {boneName=dyn_poncho_front_r_01}
    constraintSetup [{targetBone {boneName=upperarm_r}}]
    constraintWeights [1]
   }
   n72 = RigidBody(ComponentPose=@n71) {
    bClampLinearTranslationLimitToRefPose=true
    bOverrideWorldGravity=true
    overridePhysicsAsset=&/M_MED_Wombat_Poncho_Physics
    simulationSpace=BaseBoneSpace
    alphaBoolBlend {blendOption=Linear}
    baseBoneRef {boneName=spine_02}
    componentAppliedLinearAccClamp=<3000 3000 3000>
    componentLinearAccScale=<0.85 0.25 0.45>
    componentLinearVelScale=<0.35 0.2 0.2>
    overrideWorldGravity {z=-980}
   }
   n73 = "Constraint - dyn_back_strap_master (PPP)"(ComponentPose=@n72, Alpha=1, ConstraintWeights_0=0.0331, ConstraintWeights_1=0.157, ConstraintWeights_2=0.8099) {
    alphaBoolBlend {blendOption=Linear}
    boneToModify {boneName=dyn_back_strap_master}
    constraintSetup [{targetBone {boneName=spine_01}}, {targetBone {boneName=thigh_twist_01_r}}, {targetBone {boneName=pelvis}}]
    constraintWeights [1,1,1]
   }
   n74 = Get Aim_Up()
   n75 = "Transform (Modify) Bone - Bone: dyn_poncho_front_r_01"(ComponentPose=@n73, Alpha=@n74, Translation="5,10,0", Rotation="0,0,0", Scale="1,1,1") {
    rotationMode=BMM_Additive
    scaleMode=BMM_Additive
    translationMode=BMM_Additive
    alphaBoolBlend {blendOption=Linear}
    boneToModify {boneName=dyn_poncho_front_r_01}
    translation=<5 10>
   }
   n76 = "Transform (Modify) Bone - Bone: dyn_poncho_front_l_01"(ComponentPose=@n75, Alpha=@n74, Translation="-10,5,0", Rotation="0,0,0", Scale="1,1,1") {
    rotationMode=BMM_Additive
    scaleMode=BMM_Additive
    translationMode=BMM_Additive
    alphaBoolBlend {blendOption=Linear}
    boneToModify {boneName=dyn_poncho_front_l_01}
    translation=<-10 5>
   }
   n77 = "Transform (Modify) Bone - Bone: dyn_poncho_front_mr_01"(ComponentPose=@n76, Alpha=@n74, Translation="5,5,5", Rotation="0,0,0", Scale="1,1,1") {
    rotationMode=BMM_Additive
    scaleMode=BMM_Additive
    translationMode=BMM_Additive
    alphaBoolBlend {blendOption=Linear}
    boneToModify {boneName=dyn_poncho_front_mr_01}
    translation=<5 5 5>
   }
   n78 = "Transform (Modify) Bone - Bone: dyn_poncho_front_ml_01"(ComponentPose=@n77, Alpha=@n74, Translation="-5,5,5", Rotation="0,0,0", Scale="1,1,1") {
    rotationMode=BMM_Additive
    scaleMode=BMM_Additive
    translationMode=BMM_Additive
    alphaBoolBlend {blendOption=Linear}
    boneToModify {boneName=dyn_poncho_front_ml_01}
    translation=<-5 5 5>
   }
   n79 = "Transform (Modify) Bone - Bone: dyn_poncho_front_mid_01"(ComponentPose=@n78, Alpha=@n74, Translation="0,5,0", Rotation="0,0,0", Scale="1,1,1") {
    rotationMode=BMM_Additive
    scaleMode=BMM_Additive
    translationMode=BMM_Additive
    alphaBoolBlend {blendOption=Linear}
    boneToModify {boneName=dyn_poncho_front_mid_01}
    translation {y=5}
   }
   n80 = "Transform (Modify) Bone - Bone: dyn_poncho_front_r_01"(ComponentPose=@n79, Alpha=@n74, Translation="0,0,15", Rotation="0,0,0", Scale="1,1,1") {
    rotationMode=BMM_Additive
    scaleMode=BMM_Additive
    translationMode=BMM_Additive
    alphaBoolBlend {blendOption=Linear}
    boneToModify {boneName=dyn_poncho_front_r_01}
    translation {z=15}
   }
   n81 = "Transform (Modify) Bone - Bone: dyn_poncho_front_l_01"(ComponentPose=@n80, Alpha=@n74, Translation="0,0,10", Rotation="0,0,0", Scale="1,1,1") {
    rotationMode=BMM_Additive
    scaleMode=BMM_Additive
    translationMode=BMM_Additive
    alphaBoolBlend {blendOption=Linear}
    boneToModify {boneName=dyn_poncho_front_l_01}
    translation {z=10}
   }
   n82 = "Constraint - dyn_poncho_armpit_back_r (PP)"(ComponentPose=@n81, Alpha=1, ConstraintWeights_0=0.5, ConstraintWeights_1=0.5) {
    alphaBoolBlend {blendOption=Linear}
    boneToModify {boneName=dyn_poncho_armpit_back_r}
    constraintSetup [{targetBone {boneName=dyn_poncho_back_r_01}}, {targetBone {boneName=dyn_r_sleeve_back_01}}]
    constraintWeights [1,1]
   }
   n83 = Component To Local(ComponentPose=@n82)
   n84 = "Constraint - dyn_poncho_front_r_01 (P)"(ComponentPose=Local To Component(LocalPose=Copy Pose From Mesh() {bUseAttachedParent=true}), Alpha=0.9, ConstraintWeights_0=1) {
    alphaBoolBlend {blendOption=Linear}
    boneToModify {boneName=dyn_poncho_front_r_01}
    constraintSetup [{targetBone {boneName=upperarm_r}}]
    constraintWeights [1]
   }
   n85 = RigidBody(ComponentPose=@n84) {
    bClampLinearTranslationLimitToRefPose=true
    bOverrideWorldGravity=true
    overridePhysicsAsset=&/M_MED_Wombat_Poncho_Physics_Crouch
    simulationSpace=BaseBoneSpace
    alphaBoolBlend {blendOption=Linear}
    baseBoneRef {boneName=spine_02}
    componentAppliedLinearAccClamp=<3000 3000 3000>
    componentLinearAccScale=<0.85 0.25 0.5>
    componentLinearVelScale=<0.35 0.2 0.2>
    overrideWorldGravity {z=-980}
   }
   n86 = "Constraint - dyn_back_strap_master (PPP)"(ComponentPose=@n85, Alpha=1, ConstraintWeights_0=0.0331, ConstraintWeights_1=0.157, ConstraintWeights_2=0.8099) {
    alphaBoolBlend {blendOption=Linear}
    boneToModify {boneName=dyn_back_strap_master}
    constraintSetup [{targetBone {boneName=spine_01}}, {targetBone {boneName=thigh_twist_01_r}}, {targetBone {boneName=pelvis}}]
    constraintWeights [1,1,1]
   }
   n87 = "Transform (Modify) Bone - Bone: dyn_belt_knife_r_01"(ComponentPose=@n86, Alpha=1, Translation="0,0,1", Rotation="-5,0,0", Scale="1,1,1") {
    rotationMode=BMM_Additive
    scaleMode=BMM_Additive
    translationMode=BMM_Additive
    alphaBoolBlend {blendOption=Linear}
    boneToModify {boneName=dyn_belt_knife_r_01}
    rotation {pitch=-5}
    translation {z=1}
   }
   n88 = "Transform (Modify) Bone - Bone: dyn_belt_knife_l_01"(ComponentPose=@n87, Alpha=1, Translation="1,0,1", Rotation="0,0,0", Scale="1,1,1") {
    rotationMode=BMM_Additive
    scaleMode=BMM_Additive
    translationMode=BMM_Additive
    alphaBoolBlend {blendOption=Linear}
    boneToModify {boneName=dyn_belt_knife_l_01}
    translation {x=1, z=1}
   }
   n89 = "Transform (Modify) Bone - Bone: dyn_poncho_front_l_master"(ComponentPose=@n88, Alpha=1, Translation="-2,0,2", Rotation="0,0,0", Scale="1,1,1") {
    rotationMode=BMM_Additive
    scaleMode=BMM_Additive
    translationMode=BMM_Additive
    alphaBoolBlend {blendOption=Linear}
    boneToModify {boneName=dyn_poncho_front_l_master}
    translation {x=-2, z=2}
   }
   n90 = "Constraint - dyn_poncho_armpit_back_r (PP)"(ComponentPose=@n89, Alpha=1, ConstraintWeights_0=0.5, ConstraintWeights_1=0.5) {
    alphaBoolBlend {blendOption=Linear}
    boneToModify {boneName=dyn_poncho_armpit_back_r}
    constraintSetup [{targetBone {boneName=dyn_poncho_back_r_01}}, {targetBone {boneName=dyn_r_sleeve_back_01}}]
    constraintWeights [1,1]
   }
   n91 = "Transform (Modify) Bone - Bone: dyn_poncho_armpit_back_r"(ComponentPose=@n90, Alpha=1, Translation="0,-5,0", Rotation="0,0,0", Scale="1,1,1") {
    rotationMode=BMM_Additive
    scaleMode=BMM_Additive
    translationMode=BMM_Additive
    alphaBoolBlend {blendOption=Linear}
    boneToModify {boneName=dyn_poncho_armpit_back_r}
    translation {y=-5}
   }
   n92 = "Transform (Modify) Bone - Bone: dyn_poncho_back_mid_master"(ComponentPose=@n91, Alpha=1, Translation="0,0,-5", Rotation="0,0,0", Scale="1,1,1") {
    rotationMode=BMM_Additive
    scaleMode=BMM_Additive
    translationMode=BMM_Additive
    translationSpace=BCS_BoneSpace
    alphaBoolBlend {blendOption=Linear}
    boneToModify {boneName=dyn_poncho_back_mid_master}
    translation {z=-5}
   }
   n93 = "Transform (Modify) Bone - Bone: dyn_poncho_back_mr_master"(ComponentPose=@n92, Alpha=1, Translation="0,0,-5", Rotation="0,0,0", Scale="1,1,1") {
    rotationMode=BMM_Additive
    scaleMode=BMM_Additive
    translationMode=BMM_Additive
    translationSpace=BCS_BoneSpace
    alphaBoolBlend {blendOption=Linear}
    boneToModify {boneName=dyn_poncho_back_mr_master}
    translation {z=-5}
   }
   n94 = "Transform (Modify) Bone - Bone: dyn_poncho_front_l_01"(ComponentPose=@n93, Alpha=1, Translation="5,0,5", Rotation="0,0,0", Scale="1,1,1") {
    rotationMode=BMM_Additive
    scaleMode=BMM_Additive
    translationMode=BMM_Additive
    alphaBoolBlend {blendOption=Linear}
    boneToModify {boneName=dyn_poncho_front_l_01}
    translation {x=5, z=5}
   }
   n95 = "Constraint - dyn_l_shoelaces_master (PPP)"(ComponentPose=@n94, Alpha=1, ConstraintWeights_0=0.7001, ConstraintWeights_1=0.1546, ConstraintWeights_2=0.1453) {
    alphaBoolBlend {blendOption=Linear}
    boneToModify {boneName=dyn_l_shoelaces_master}
    constraintSetup [{targetBone {boneName=calf_l}}, {targetBone {boneName=calf_twist_02_l}}, {targetBone {boneName=thigh_l}}]
    constraintWeights [1,1,1]
   }
   n96 = "Constraint - dyn_l_knee_string_long_master (PPP)"(ComponentPose=@n95, Alpha=1, ConstraintWeights_0=0.7001, ConstraintWeights_1=0.1546, ConstraintWeights_2=0.1453) {
    alphaBoolBlend {blendOption=Linear}
    boneToModify {boneName=dyn_l_knee_string_long_master}
    constraintSetup [{targetBone {boneName=calf_l}}, {targetBone {boneName=calf_twist_02_l}}, {targetBone {boneName=thigh_l}}]
    constraintWeights [1,1,1]
   }
   n97 = Component To Local(ComponentPose=@n96)
   n98 = Get bIsCrouching()
   n99 = "Constraint - dyn_poncho_armpit_back_l (PP)"(
    ComponentPose=Local To Component(LocalPose=Blend Poses by bool(BlendPose_0=@n97, BlendPose_1=@n83, bActiveValue=@n98) {blendType=Linear})
    Alpha=1
    ConstraintWeights_0=0.5
    ConstraintWeights_1=0.5
   ) {
    alphaBoolBlend {blendOption=Linear}
    boneToModify {boneName=dyn_poncho_armpit_back_l}
    constraintSetup [{targetBone {boneName=dyn_poncho_back_l_01}}, {targetBone {boneName=dyn_l_sleeve_back_01}}]
    constraintWeights [1,1]
   }
   n100 = Get Aimup_Off() -> (Aimup_Off:double)
   n101 = "Constraint - dyn_poncho_armpit_front_r (PP)"(ComponentPose=@n99, Alpha=@n100, ConstraintWeights_0=0.5, ConstraintWeights_1=0.5) {
    alphaBoolBlend {blendOption=Linear}
    boneToModify {boneName=dyn_poncho_armpit_front_r}
    constraintSetup [{targetBone {boneName=dyn_poncho_front_r_01}}, {targetBone {boneName=upperarm_r}}]
    constraintWeights [1,1]
   }
   n102 = "Constraint - dyn_poncho_armpit_front_l (PP)"(ComponentPose=@n101, Alpha=@n100, ConstraintWeights_0=0.5, ConstraintWeights_1=0.5) {
    alphaBoolBlend {blendOption=Linear}
    boneToModify {boneName=dyn_poncho_armpit_front_l}
    constraintSetup [{targetBone {boneName=dyn_poncho_front_l_01}}, {targetBone {boneName=upperarm_l}}]
    constraintWeights [1,1]
   }
   n103 = "Constraint - dyn_poncho_armpit_front_r (P)"(ComponentPose=@n102, Alpha=1, ConstraintWeights_0=1) {
    alphaBoolBlend {blendOption=Linear}
    boneToModify {boneName=dyn_poncho_armpit_front_r}
    constraintSetup [{targetBone {boneName=upperarm_r}}]
    constraintWeights [1]
   }
   n104 = "Transform (Modify) Bone - Bone: dyn_poncho_armpit_front_r"(ComponentPose=@n103, Alpha=Get Aim_Up(), Translation="0,5,-5", Rotation="-40,40,0", Scale="1,1,1") {
    rotationMode=BMM_Additive
    scaleMode=BMM_Additive
    translationMode=BMM_Additive
    alphaBoolBlend {blendOption=Linear}
    boneToModify {boneName=dyn_poncho_armpit_front_r}
    rotation {pitch=-40, yaw=40}
    translation {y=5, z=-5}
   }
   n105 = Get Aim_Up()
   n106 = "Constraint - dyn_l_sleeve_front_01 (P)"(ComponentPose=@n104, Alpha=@n105, ConstraintWeights_0=1) {
    alphaBoolBlend {blendOption=Linear}
    boneToModify {boneName=dyn_l_sleeve_front_01}
    constraintSetup [{targetBone {boneName=upperarm_l}}]
    constraintWeights [1]
   }
   n107 = "Constraint - dyn_r_sleeve_front_01 (P)"(ComponentPose=@n106, Alpha=@n105, ConstraintWeights_0=1) {
    alphaBoolBlend {blendOption=Linear}
    boneToModify {boneName=dyn_r_sleeve_front_01}
    constraintSetup [{targetBone {boneName=upperarm_r}}]
    constraintWeights [1]
   }
   n108 = "Transform (Modify) Bone - Bone: dyn_r_sleeve_front_01"(ComponentPose=@n107, Alpha=@n105, Translation="2,0,-5", Rotation="0,0,0", Scale="1,1,1") {
    rotationMode=BMM_Additive
    scaleMode=BMM_Additive
    translationMode=BMM_Additive
    alphaBoolBlend {blendOption=Linear}
    boneToModify {boneName=dyn_r_sleeve_front_01}
    translation {x=2, z=-5}
   }
   n109 = "Transform (Modify) Bone - Bone: dyn_poncho_front_r_01"(ComponentPose=@n108, Alpha=@n105, Translation="0,0,5", Rotation="0,0,0", Scale="1,1,1") {
    rotationMode=BMM_Additive
    scaleMode=BMM_Additive
    translationMode=BMM_Additive
    alphaBoolBlend {blendOption=Linear}
    boneToModify {boneName=dyn_poncho_front_r_01}
    translation {z=5}
   }
   n110 = Get Aim_Up()
   n111 = "Transform (Modify) Bone - Bone: dyn_r_sleeve_front_01"(ComponentPose=@n109, Alpha=@n110, Translation="0,0,1", Rotation="0,0,-5", Scale="1,1,1") {
    rotationMode=BMM_Additive
    scaleMode=BMM_Additive
    translationMode=BMM_Additive
    alphaBoolBlend {blendOption=Linear}
    boneToModify {boneName=dyn_r_sleeve_front_01}
    rotation {roll=-5}
    translation {z=1}
   }
   n112 = "Transform (Modify) Bone - Bone: dyn_r_sleeve_mid_01"(ComponentPose=@n111, Alpha=@n110, Translation="0,0,1", Rotation="40,0,0", Scale="1,1,1") {
    rotationMode=BMM_Additive
    scaleMode=BMM_Additive
    translationMode=BMM_Additive
    alphaBoolBlend {blendOption=Linear}
    boneToModify {boneName=dyn_r_sleeve_mid_01}
    rotation {pitch=40}
    translation {z=1}
   }
   n113 = "Transform (Modify) Bone - Bone: dyn_r_sleeve_back_01"(ComponentPose=@n112, Alpha=@n110, Translation="-5,0,0", Rotation="0,0,0", Scale="1,1,1") {
    rotationMode=BMM_Additive
    scaleMode=BMM_Additive
    translationMode=BMM_Additive
    alphaBoolBlend {blendOption=Linear}
    boneToModify {boneName=dyn_r_sleeve_back_01}
    translation {x=-5}
   }
   n114 = "Transform (Modify) Bone - Bone: dyn_r_sleeve_back_01"(ComponentPose=@n113, Alpha=@n110, Translation="0,0,0", Rotation="0,-80,0", Scale="1,1,1") {
    rotationMode=BMM_Additive
    scaleMode=BMM_Additive
    translationMode=BMM_Additive
    alphaBoolBlend {blendOption=Linear}
    boneToModify {boneName=dyn_r_sleeve_back_01}
    rotation {yaw=-80}
   }
   n115 = "Transform (Modify) Bone - Bone: dyn_r_sleeve_mid_01"(ComponentPose=@n114, Alpha=@n110, Translation="0,0,0", Rotation="-30,0,0", Scale="1,1,1") {
    rotationMode=BMM_Additive
    scaleMode=BMM_Additive
    translationMode=BMM_Additive
    alphaBoolBlend {blendOption=Linear}
    boneToModify {boneName=dyn_r_sleeve_mid_01}
    rotation {pitch=-30}
   }
   n116 = Get Aim_Up()
   n117 = "Constraint - dyn_poncho_armpit_front_l (P)"(ComponentPose=@n115, Alpha=@n116, ConstraintWeights_0=1) {
    alphaBoolBlend {blendOption=Linear}
    boneToModify {boneName=dyn_poncho_armpit_front_l}
    constraintSetup [{targetBone {boneName=upperarm_l}}]
    constraintWeights [1]
   }
   n118 = "Transform (Modify) Bone - Bone: dyn_poncho_armpit_front_l"(ComponentPose=@n117, Alpha=@n116, Translation="10,-5,-10", Rotation="30,30,0", Scale="1,1,1") {
    rotationMode=BMM_Additive
    scaleMode=BMM_Additive
    translationMode=BMM_Additive
    alphaBoolBlend {blendOption=Linear}
    boneToModify {boneName=dyn_poncho_armpit_front_l}
    rotation {pitch=30, yaw=30}
    translation=<10 -5 -10>
   }
   n119 = "Transform (Modify) Bone - Bone: dyn_l_sleeve_front_01"(ComponentPose=@n118, Alpha=@n116, Translation="0,0,0", Rotation="20,0,-20", Scale="1,1,1") {
    rotationMode=BMM_Additive
    scaleMode=BMM_Additive
    translationMode=BMM_Additive
    alphaBoolBlend {blendOption=Linear}
    boneToModify {boneName=dyn_l_sleeve_front_01}
    rotation {pitch=20, roll=-20}
   }
   n120 = "Transform (Modify) Bone - Bone: dyn_l_sleeve_back_01"(ComponentPose=@n119, Alpha=@n116, Translation="5,0,0", Rotation="40,0,-40", Scale="1,1,1") {
    rotationMode=BMM_Additive
    scaleMode=BMM_Additive
    translationMode=BMM_Additive
    alphaBoolBlend {blendOption=Linear}
    boneToModify {boneName=dyn_l_sleeve_back_01}
    rotation {pitch=40, roll=-40}
    translation {x=5}
   }
   n121 = "Transform (Modify) Bone - Bone: dyn_poncho_armpit_back_l"(ComponentPose=@n120, Alpha=@n116, Translation="0,0,10", Rotation="0,0,0", Scale="1,1,1") {
    rotationMode=BMM_Additive
    scaleMode=BMM_Additive
    translationMode=BMM_Additive
    alphaBoolBlend {blendOption=Linear}
    boneToModify {boneName=dyn_poncho_armpit_back_l}
    translation {z=10}
   }
   n122 = "Transform (Modify) Bone - Bone: dyn_poncho_back_l_01"(ComponentPose=@n121, Alpha=@n116, Translation="0,0,4", Rotation="0,0,0", Scale="1,1,1") {
    rotationMode=BMM_Additive
    scaleMode=BMM_Additive
    translationMode=BMM_Additive
    alphaBoolBlend {blendOption=Linear}
    boneToModify {boneName=dyn_poncho_back_l_01}
    translation {z=4}
   }
   n123 = "Transform (Modify) Bone - Bone: dyn_r_sleeve_back_01"(ComponentPose=@n122, Alpha=@n116, Translation="-10,0,0", Rotation="0,0,0", Scale="1,1,1") {
    rotationMode=BMM_Additive
    scaleMode=BMM_Additive
    translationMode=BMM_Additive
    alphaBoolBlend {blendOption=Linear}
    boneToModify {boneName=dyn_r_sleeve_back_01}
    translation {x=-10}
   }
   n124 = "Transform (Modify) Bone - Bone: dyn_poncho_armpit_back_r"(ComponentPose=@n123, Alpha=1, Translation="-5,0,0", Rotation="0,0,0", Scale="1,1,1") {
    rotationMode=BMM_Additive
    scaleMode=BMM_Additive
    translationMode=BMM_Additive
    alphaBoolBlend {blendOption=Linear}
    boneToModify {boneName=dyn_poncho_armpit_back_r}
    translation {x=-5}
   }
   n125 = "Constraint - dyn_poncho_armpit_front_l (P)"(ComponentPose=@n124, Alpha=1, ConstraintWeights_0=1) {
    alphaBoolBlend {blendOption=Linear}
    boneToModify {boneName=dyn_poncho_armpit_front_l}
    constraintSetup [{targetBone {boneName=upperarm_l}}]
    constraintWeights [1]
   }
   n126 = "Transform (Modify) Bone - Bone: dyn_l_sleeve_front_01"(ComponentPose=@n125, Alpha=Get Rifle_Adjustment() -> (Rifle_Adjustment:double), Translation="-2,-2,2", Rotation="0,0,0", Scale="1,1,1") {
    rotationMode=BMM_Additive
    scaleMode=BMM_Additive
    translationMode=BMM_Additive
    alphaBoolBlend {blendOption=Linear}
    boneToModify {boneName=dyn_l_sleeve_front_01}
    translation=<-2 -2 2>
   }
   n127 = "Transform (Modify) Bone - Bone: dyn_poncho_armpit_front_l"(ComponentPose=@n126, Alpha=Get Rifle_Mid() -> (Rifle_Mid:double), Translation="-10,0,0", Rotation="0,0,0", Scale="1,1,1") {
    rotationMode=BMM_Additive
    scaleMode=BMM_Additive
    translationMode=BMM_Additive
    alphaBoolBlend {blendOption=Linear}
    boneToModify {boneName=dyn_poncho_armpit_front_l}
    translation {x=-10}
   }
   n128 = "Transform (Modify) Bone - Bone: dyn_poncho_armpit_front_l"(ComponentPose=@n127, Alpha=Get Aim_Up_Variable() -> (Aim_Up_Variable:double), Translation="-2,0,0", Rotation="0,0,0", Scale="1,1,1") {
    rotationMode=BMM_Additive
    scaleMode=BMM_Additive
    translationMode=BMM_Additive
    alphaBoolBlend {blendOption=Linear}
    boneToModify {boneName=dyn_poncho_armpit_front_l}
    translation {x=-2}
   }
   n129 = "Transform (Modify) Bone - Bone: dyn_poncho_front_l_01"(ComponentPose=@n128, Alpha=Get Aim_Up_Variable(), Translation="-10,5,0", Rotation="0,0,0", Scale="1,1,1") {
    rotationMode=BMM_Additive
    scaleMode=BMM_Additive
    translationMode=BMM_Additive
    alphaBoolBlend {blendOption=Linear}
    boneToModify {boneName=dyn_poncho_front_l_01}
    translation=<-10 5>
   }
   n130 = "Transform (Modify) Bone - Bone: dyn_poncho_back_r_02"(ComponentPose=@n129, Alpha=Get Aim_Up(), Translation="0,5,0", Rotation="0,0,0", Scale="1,1,1") {
    rotationMode=BMM_Additive
    scaleMode=BMM_Additive
    translationMode=BMM_Additive
    alphaBoolBlend {blendOption=Linear}
    boneToModify {boneName=dyn_poncho_back_r_02}
    translation {y=5}
   }
   n131 = "Transform (Modify) Bone - Bone: dyn_poncho_back_ml_01"(ComponentPose=@n130, Alpha=1, Translation="0,-5,1", Rotation="0,0,0", Scale="1,1,1") {
    rotationMode=BMM_Additive
    scaleMode=BMM_Additive
    translationMode=BMM_Additive
    alphaBoolBlend {blendOption=Linear}
    boneToModify {boneName=dyn_poncho_back_ml_01}
    translation {y=-5, z=1}
   }
   n132 = "Transform (Modify) Bone - Bone: dyn_poncho_back_l_01"(ComponentPose=@n131, Alpha=1, Translation="0,-5,1", Rotation="0,0,0", Scale="1,1,1") {
    rotationMode=BMM_Additive
    scaleMode=BMM_Additive
    translationMode=BMM_Additive
    alphaBoolBlend {blendOption=Linear}
    boneToModify {boneName=dyn_poncho_back_l_01}
    translation {y=-5, z=1}
   }
   n133 = "Transform (Modify) Bone - Bone: dyn_poncho_armpit_back_l"(ComponentPose=@n132, Alpha=1, Translation="0,-5,1", Rotation="0,0,0", Scale="1,1,1") {
    rotationMode=BMM_Additive
    scaleMode=BMM_Additive
    translationMode=BMM_Additive
    alphaBoolBlend {blendOption=Linear}
    boneToModify {boneName=dyn_poncho_armpit_back_l}
    translation {y=-5, z=1}
   }
   n134 = "Transform (Modify) Bone - Bone: dyn_l_sleeve_back_01"(ComponentPose=@n133, Alpha=1, Translation="0,-8,1", Rotation="0,0,0", Scale="1,1,1") {
    rotationMode=BMM_Additive
    scaleMode=BMM_Additive
    translationMode=BMM_Additive
    alphaBoolBlend {blendOption=Linear}
    boneToModify {boneName=dyn_l_sleeve_back_01}
    translation {y=-8, z=1}
   }
   n135 = "Constraint - dyn_poncho_armpit_back_l (PP)"(ComponentPose=@n134, Alpha=1, ConstraintWeights_0=0.5, ConstraintWeights_1=0.5) {
    alphaBoolBlend {blendOption=Linear}
    boneToModify {boneName=dyn_poncho_armpit_back_l}
    constraintSetup [{targetBone {boneName=upperarm_l}}, {targetBone {boneName=dyn_poncho_back_l_01}}]
    constraintWeights [1,1]
   }
   n136 = "Transform (Modify) Bone - Bone: dyn_poncho_armpit_back_l"(ComponentPose=@n135, Alpha=@n5, Translation="8,0,-2", Rotation="60,40,0", Scale="1,1,1") {
    rotationMode=BMM_Additive
    scaleMode=BMM_Additive
    translationMode=BMM_Additive
    alphaBoolBlend {blendOption=Linear}
    boneToModify {boneName=dyn_poncho_armpit_back_l}
    rotation {pitch=60, yaw=40}
    translation {x=8, z=-2}
   }
   n137 = "Transform (Modify) Bone - Bone: dyn_poncho_back_ml_01"(ComponentPose=@n136, Alpha=@n5, Translation="4,4,2", Rotation="0,0,0", Scale="1,1,1") {
    rotationMode=BMM_Additive
    scaleMode=BMM_Additive
    translationMode=BMM_Additive
    alphaBoolBlend {blendOption=Linear}
    boneToModify {boneName=dyn_poncho_back_ml_01}
    translation=<4 4 2>
   }
   n138 = "Transform (Modify) Bone - Bone: dyn_l_sleeve_back_01"(ComponentPose=@n137, Alpha=1, Translation="4,4,2", Rotation="0,0,0", Scale="1,1,1") {
    rotationMode=BMM_Additive
    scaleMode=BMM_Additive
    translationMode=BMM_Additive
    alphaBoolBlend {blendOption=Linear}
    boneToModify {boneName=dyn_l_sleeve_back_01}
    translation=<4 4 2>
   }
   n139 = "Constraint - dyn_poncho_armpit_front_l (P)"(ComponentPose=@n138, Alpha=Get Aim_Up(), ConstraintWeights_0=1) {
    alphaBoolBlend {blendOption=Linear}
    boneToModify {boneName=dyn_poncho_armpit_front_l}
    constraintSetup [{targetBone {boneName=upperarm_l}}]
    constraintWeights [1]
   }
   n140 = Get Sprint_Float() -> (Sprint_Float:double)
   n141 = "Transform (Modify) Bone - Bone: dyn_poncho_back_r_master"(ComponentPose=@n139, Alpha=@n140, Translation="0,-8,2", Rotation="0,0,0", Scale="1,1,1") {
    rotationMode=BMM_Additive
    scaleMode=BMM_Additive
    translationMode=BMM_Additive
    alphaBoolBlend {blendOption=Linear}
    boneToModify {boneName=dyn_poncho_back_r_master}
    translation {y=-8, z=2}
   }
   n142 = "Transform (Modify) Bone - Bone: dyn_r_sleeve_back_01"(ComponentPose=@n141, Alpha=@n140, Translation="0,-5,0", Rotation="0,0,0", Scale="1,1,1") {
    rotationMode=BMM_Additive
    scaleMode=BMM_Additive
    translationMode=BMM_Additive
    alphaBoolBlend {blendOption=Linear}
    boneToModify {boneName=dyn_r_sleeve_back_01}
    translation {y=-5}
   }
   n143 = "Transform (Modify) Bone - Bone: dyn_poncho_back_l_master"(ComponentPose=@n142, Alpha=@n140, Translation="0,-5,2", Rotation="0,0,0", Scale="1,1,1") {
    rotationMode=BMM_Additive
    scaleMode=BMM_Additive
    translationMode=BMM_Additive
    alphaBoolBlend {blendOption=Linear}
    boneToModify {boneName=dyn_poncho_back_l_master}
    translation {y=-5, z=2}
   }
   n144 = "Transform (Modify) Bone - Bone: dyn_poncho_armpit_back_l"(ComponentPose=@n143, Alpha=@n140, Translation="0,-10,0", Rotation="0,0,0", Scale="1,1,1") {
    rotationMode=BMM_Additive
    scaleMode=BMM_Additive
    translationMode=BMM_Additive
    alphaBoolBlend {blendOption=Linear}
    boneToModify {boneName=dyn_poncho_armpit_back_l}
    translation {y=-10}
   }
   n145 = "Transform (Modify) Bone - Bone: dyn_l_sleeve_back_01"(ComponentPose=@n144, Alpha=@n140, Translation="0,-5,0", Rotation="0,0,0", Scale="1,1,1") {
    rotationMode=BMM_Additive
    scaleMode=BMM_Additive
    translationMode=BMM_Additive
    alphaBoolBlend {blendOption=Linear}
    boneToModify {boneName=dyn_l_sleeve_back_01}
    translation {y=-5}
   }
   n146 = Anim Dynamics - Bone: dyn_l_knee_string_long_01!(ComponentPose=@n145, Alpha=1, GravityOverride:Vector="0,0,0", ExternalForce="0,0,0") -> (Pose) {
    angularDampingOverride=1
    angularSpringConstant=1500
    bAngularSpring=true
    bGravityOverrideInSimSpace=true
    bOverrideAngularDamping=true
    gravityScale=0.35
    simulationSpace=RootRelative
    alphaBoolBlend {blendOption=Linear}
    alphaScaleBias {scale=2}
    boundBone {boneName=dyn_l_knee_string_long_01}
    physicsBodyDefinitions [{
      sphereCollisionRadius=0
      boundBone {boneName=dyn_l_knee_string_long_01}
      boxExtents=<10 20 20>
      constraintSetup {angularLimitsMax=<0 2 5>, angularLimitsMin=<0 -20 -5>, angularTarget=<0 0 0>}
      localJointOffset=<10 0 0>
     }]
   }
   n147 = Anim Dynamics - Bone: dyn_l_knee_string_long_02!(ComponentPose=@n146, Alpha=1, GravityOverride:Vector="0,0,0", ExternalForce="0,0,0") -> (Pose) {
    angularDampingOverride=1
    angularSpringConstant=1500
    bAngularSpring=true
    bGravityOverrideInSimSpace=true
    bOverrideAngularDamping=true
    gravityScale=0.35
    simulationSpace=RootRelative
    alphaBoolBlend {blendOption=Linear}
    alphaScaleBias {scale=2}
    boundBone {boneName=dyn_l_knee_string_long_02}
    physicsBodyDefinitions [{
      sphereCollisionRadius=0
      boundBone {boneName=dyn_l_knee_string_long_02}
      boxExtents=<10 20 20>
      constraintSetup {angularLimitsMax=<0 2 5>, angularLimitsMin=<0 -20 -5>, angularTarget=<0 0 0>}
      localJointOffset=<10 0 0>
     }]
   }
   n148 = Anim Dynamics - Bone: dyn_l_knee_string_long_03!(ComponentPose=@n147, Alpha=1, GravityOverride:Vector="0,0,0", ExternalForce="0,0,0") -> (Pose) {
    angularDampingOverride=1
    angularSpringConstant=1500
    bAngularSpring=true
    bGravityOverrideInSimSpace=true
    bOverrideAngularDamping=true
    gravityScale=0.35
    simulationSpace=RootRelative
    alphaBoolBlend {blendOption=Linear}
    alphaScaleBias {scale=2}
    boundBone {boneName=dyn_l_knee_string_long_03}
    physicsBodyDefinitions [{
      sphereCollisionRadius=0
      boundBone {boneName=dyn_l_knee_string_long_03}
      boxExtents=<10 20 20>
      constraintSetup {angularLimitsMax=<0 2 5>, angularLimitsMin=<0 -20 -5>, angularTarget=<0 0 0>}
      localJointOffset=<10 0 0>
     }]
   }
   n149 = Anim Dynamics - Bone: dyn_r_knee_string_01!(ComponentPose=@n148, Alpha=1, GravityOverride:Vector="0,0,0", ExternalForce="0,0,0") -> (Pose) {
    angularDampingOverride=1
    angularSpringConstant=1500
    bAngularSpring=true
    bGravityOverrideInSimSpace=true
    bOverrideAngularDamping=true
    gravityScale=0.35
    simulationSpace=RootRelative
    alphaBoolBlend {blendOption=Linear}
    alphaScaleBias {scale=2}
    boundBone {boneName=dyn_r_knee_string_01}
    physicsBodyDefinitions [{
      sphereCollisionRadius=0
      boundBone {boneName=dyn_r_knee_string_01}
      boxExtents=<10 20 20>
      constraintSetup {angularLimitsMax=<0 0 5>, angularLimitsMin=<0 -20 -5>, angularTarget=<0 0 0>}
      localJointOffset=<10 0 0>
     }]
   }
   n150 = Anim Dynamics - Bone: dyn_r_knee_string_02!(ComponentPose=@n149, Alpha=1, GravityOverride:Vector="0,0,0", ExternalForce="0,0,0") -> (Pose) {
    angularDampingOverride=1
    angularSpringConstant=1500
    bAngularSpring=true
    bGravityOverrideInSimSpace=true
    bOverrideAngularDamping=true
    gravityScale=0.35
    simulationSpace=RootRelative
    alphaBoolBlend {blendOption=Linear}
    alphaScaleBias {scale=2}
    boundBone {boneName=dyn_r_knee_string_02}
    physicsBodyDefinitions [{
      sphereCollisionRadius=0
      boundBone {boneName=dyn_r_knee_string_02}
      boxExtents=<10 20 20>
      constraintSetup {angularLimitsMax=<0 0 5>, angularLimitsMin=<0 -20 -5>, angularTarget=<0 0 0>}
      localJointOffset=<10 0 0>
     }]
   }
   n151 = Anim Dynamics - Bone: dyn_r_knee_string_03!(ComponentPose=@n150, Alpha=1, GravityOverride:Vector="0,0,0", ExternalForce="0,0,0") -> (Pose) {
    angularDampingOverride=1
    angularSpringConstant=1500
    bAngularSpring=true
    bGravityOverrideInSimSpace=true
    bOverrideAngularDamping=true
    gravityScale=0.35
    simulationSpace=RootRelative
    alphaBoolBlend {blendOption=Linear}
    alphaScaleBias {scale=2}
    boundBone {boneName=dyn_r_knee_string_03}
    physicsBodyDefinitions [{
      sphereCollisionRadius=0
      boundBone {boneName=dyn_r_knee_string_03}
      boxExtents=<10 20 20>
      constraintSetup {angularLimitsMax=<0 0 5>, angularLimitsMin=<0 -20 -5>, angularTarget=<0 0 0>}
      localJointOffset=<10 0 0>
     }]
   }
   n152 = Anim Dynamics - Bone: dyn_r_string_r_01!(ComponentPose=@n151, Alpha=1, GravityOverride:Vector="0,0,0", ExternalForce="0,0,0") -> (Pose) {
    angularDampingOverride=1
    angularSpringConstant=1500
    bAngularSpring=true
    bGravityOverrideInSimSpace=true
    bOverrideAngularDamping=true
    gravityScale=0.35
    simulationSpace=RootRelative
    alphaBoolBlend {blendOption=Linear}
    alphaScaleBias {scale=2}
    boundBone {boneName=dyn_r_string_r_01}
    physicsBodyDefinitions [{
      sphereCollisionRadius=0
      boundBone {boneName=dyn_r_string_r_01}
      boxExtents=<10 20 20>
      constraintSetup {angularLimitsMax=<0 0 5>, angularLimitsMin=<0 -20 -5>, angularTarget=<0 0 0>}
      localJointOffset=<10 0 0>
     }]
   }
   n153 = Anim Dynamics - Bone: dyn_r_string_r_02!(ComponentPose=@n152, Alpha=1, GravityOverride:Vector="0,0,0", ExternalForce="0,0,0") -> (Pose) {
    angularDampingOverride=1
    angularSpringConstant=1500
    bAngularSpring=true
    bGravityOverrideInSimSpace=true
    bOverrideAngularDamping=true
    gravityScale=0.35
    simulationSpace=RootRelative
    alphaBoolBlend {blendOption=Linear}
    alphaScaleBias {scale=2}
    boundBone {boneName=dyn_r_string_r_02}
    physicsBodyDefinitions [{
      sphereCollisionRadius=0
      boundBone {boneName=dyn_r_string_r_02}
      boxExtents=<10 20 20>
      constraintSetup {angularLimitsMax=<0 0 5>, angularLimitsMin=<0 -20 -5>, angularTarget=<0 0 0>}
      localJointOffset=<10 0 0>
     }]
   }
   n154 = Anim Dynamics - Bone: dyn_l_knee_string_r_loop_01!(ComponentPose=@n153, Alpha=1, GravityOverride:Vector="0,0,0", ExternalForce="0,0,0") -> (Pose) {
    angularDampingOverride=1
    angularSpringConstant=1500
    bAngularSpring=true
    bGravityOverrideInSimSpace=true
    bOverrideAngularDamping=true
    gravityScale=0.35
    simulationSpace=RootRelative
    alphaBoolBlend {blendOption=Linear}
    alphaScaleBias {scale=2}
    boundBone {boneName=dyn_l_knee_string_r_loop_01}
    physicsBodyDefinitions [{
      sphereCollisionRadius=0
      boundBone {boneName=dyn_l_knee_string_r_loop_01}
      boxExtents=<10 20 20>
      constraintSetup {angularLimitsMax=<0 0 5>, angularLimitsMin=<0 -20 -5>, angularTarget=<0 0 0>}
      localJointOffset=<10 0 0>
     }]
   }
   n155 = Anim Dynamics - Bone: dyn_l_knee_string_r_loop_02!(ComponentPose=@n154, Alpha=1, GravityOverride:Vector="0,0,0", ExternalForce="0,0,0") -> (Pose) {
    angularDampingOverride=1
    angularSpringConstant=1500
    bAngularSpring=true
    bGravityOverrideInSimSpace=true
    bOverrideAngularDamping=true
    gravityScale=0.35
    simulationSpace=RootRelative
    alphaBoolBlend {blendOption=Linear}
    alphaScaleBias {scale=2}
    boundBone {boneName=dyn_l_knee_string_r_loop_02}
    physicsBodyDefinitions [{
      sphereCollisionRadius=0
      boundBone {boneName=dyn_l_knee_string_r_loop_02}
      boxExtents=<10 20 20>
      constraintSetup {angularLimitsMax=<0 0 5>, angularLimitsMin=<0 -20 -5>, angularTarget=<0 0 0>}
      localJointOffset=<10 0 0>
     }]
   }
   n156 = Anim Dynamics - Bone: dyn_l_knee_string_r_loop_03!(ComponentPose=@n155, Alpha=1, GravityOverride:Vector="0,0,0", ExternalForce="0,0,0") -> (Pose) {
    angularDampingOverride=1
    angularSpringConstant=1500
    bAngularSpring=true
    bGravityOverrideInSimSpace=true
    bOverrideAngularDamping=true
    gravityScale=0.35
    simulationSpace=RootRelative
    alphaBoolBlend {blendOption=Linear}
    alphaScaleBias {scale=2}
    boundBone {boneName=dyn_l_knee_string_r_loop_03}
    physicsBodyDefinitions [{
      sphereCollisionRadius=0
      boundBone {boneName=dyn_l_knee_string_r_loop_03}
      boxExtents=<10 20 20>
      constraintSetup {angularLimitsMax=<0 0 5>, angularLimitsMin=<0 -20 -5>, angularTarget=<0 0 0>}
      localJointOffset=<10 0 0>
     }]
   }
   n157 = "Constraint - dyn_belt_knife_l_master (R)"(ComponentPose=@n156, Alpha=1, ConstraintWeights_0=1) {
    alphaBoolBlend {blendOption=Linear}
    boneToModify {boneName=dyn_belt_knife_l_master}
    constraintSetup [{transformType=Rotation, targetBone {boneName=thigh_l}}]
    constraintWeights [1]
   }
   n158 = "Constraint - dyn_belt_knife_r_master (R)"(ComponentPose=@n157, Alpha=0.7, ConstraintWeights_0=1) {
    alphaBoolBlend {blendOption=Linear}
    boneToModify {boneName=dyn_belt_knife_r_master}
    constraintSetup [{transformType=Rotation, targetBone {boneName=thigh_r}}]
    constraintWeights [1]
   }
   n159 = Anim Dynamics - Bone: dyn_belt_knife_l_01(ComponentPose=@n158, Alpha=1, ExternalForce="0,0,0") {
    angularDampingOverride=0.9999
    angularSpringConstant=100
    bAngularSpring=true
    bGravityOverrideInSimSpace=true
    bOverrideAngularDamping=true
    simulationSpace=BoneRelative
    alphaBoolBlend {blendOption=Linear}
    boundBone {boneName=dyn_belt_knife_l_01}
    componentLinearAccScale=<1 1 1>
    componentLinearVelScale=<1 1 1>
    physicsBodyDefinitions [{
      sphereCollisionRadius=0
      boundBone {boneName=dyn_belt_knife_l_01}
      boxExtents=<12 22 22>
      constraintSetup {angularLimitsMax=<25 0 50>, angularLimitsMin=<-25 -10 -50>, angularTarget=<0 0 0>}
      localJointOffset=<2 0 0>
     }]
    relativeSpaceBone {boneName=root}
   }
   n160 = Anim Dynamics - Bone: dyn_belt_knife_l_02(ComponentPose=@n159, Alpha=1, ExternalForce="0,0,0") {
    angularDampingOverride=0.9999
    angularSpringConstant=100
    bAngularSpring=true
    bGravityOverrideInSimSpace=true
    bOverrideAngularDamping=true
    simulationSpace=BoneRelative
    alphaBoolBlend {blendOption=Linear}
    boundBone {boneName=dyn_belt_knife_l_02}
    componentLinearAccScale=<1 1 1>
    componentLinearVelScale=<1 1 1>
    physicsBodyDefinitions [{
      sphereCollisionRadius=0
      boundBone {boneName=dyn_belt_knife_l_02}
      boxExtents=<12 22 22>
      constraintSetup {angularLimitsMax=<25 0 50>, angularLimitsMin=<-25 -10 -50>, angularTarget=<0 0 0>}
      localJointOffset=<2 0 0>
     }]
    relativeSpaceBone {boneName=root}
   }
   n161 = Anim Dynamics - Bone: dyn_belt_knife_l_03(ComponentPose=@n160, Alpha=1, ExternalForce="0,0,0") {
    angularDampingOverride=0.9999
    angularSpringConstant=100
    bAngularSpring=true
    bGravityOverrideInSimSpace=true
    bOverrideAngularDamping=true
    simulationSpace=BoneRelative
    alphaBoolBlend {blendOption=Linear}
    boundBone {boneName=dyn_belt_knife_l_03}
    componentLinearAccScale=<1 1 1>
    componentLinearVelScale=<1 1 1>
    physicsBodyDefinitions [{
      sphereCollisionRadius=0
      boundBone {boneName=dyn_belt_knife_l_03}
      boxExtents=<12 22 22>
      constraintSetup {angularLimitsMax=<25 0 50>, angularLimitsMin=<-25 -10 -50>, angularTarget=<0 0 0>}
      localJointOffset=<2 0 0>
     }]
    relativeSpaceBone {boneName=root}
   }
   n162 = Anim Dynamics - Bone: dyn_belt_knife_r_01(ComponentPose=@n161, Alpha=1, ExternalForce="0,0,0") {
    angularDampingOverride=0.9999
    angularSpringConstant=150
    bAngularSpring=true
    bGravityOverrideInSimSpace=true
    bOverrideAngularDamping=true
    simulationSpace=BoneRelative
    alphaBoolBlend {blendOption=Linear}
    boundBone {boneName=dyn_belt_knife_r_01}
    componentAppliedLinearAccClamp=<5000 5000 5000>
    componentLinearAccScale=<1 1 1>
    componentLinearVelScale=<1 1 1>
    physicsBodyDefinitions [{
      sphereCollisionRadius=0
      boundBone {boneName=dyn_belt_knife_r_01}
      boxExtents=<12 22 22>
      constraintSetup {angularLimitsMax=<25 0 50>, angularLimitsMin=<-25 -10 -50>, angularTarget=<0 0 0>}
      localJointOffset=<1 0 0>
     }]
    relativeSpaceBone {boneName=root}
   }
   n163 = Anim Dynamics - Bone: dyn_belt_knife_r_02(ComponentPose=@n162, Alpha=1, ExternalForce="0,0,0") {
    angularDampingOverride=0.9999
    angularSpringConstant=150
    bAngularSpring=true
    bGravityOverrideInSimSpace=true
    bOverrideAngularDamping=true
    simulationSpace=BoneRelative
    alphaBoolBlend {blendOption=Linear}
    boundBone {boneName=dyn_belt_knife_r_02}
    componentAppliedLinearAccClamp=<5000 5000 5000>
    componentLinearAccScale=<1 1 1>
    componentLinearVelScale=<1 1 1>
    physicsBodyDefinitions [{
      sphereCollisionRadius=0
      boundBone {boneName=dyn_belt_knife_r_02}
      boxExtents=<12 22 22>
      constraintSetup {angularLimitsMax=<25 0 50>, angularLimitsMin=<-25 -10 -50>, angularTarget=<0 0 0>}
      localJointOffset=<1 0 0>
     }]
    relativeSpaceBone {boneName=root}
   }
   n164 = Anim Dynamics - Bone: dyn_belt_knife_r_02(ComponentPose=@n163, Alpha=1, ExternalForce="0,0,0") {
    angularDampingOverride=0.9999
    angularSpringConstant=150
    bAngularSpring=true
    bGravityOverrideInSimSpace=true
    bOverrideAngularDamping=true
    simulationSpace=BoneRelative
    alphaBoolBlend {blendOption=Linear}
    boundBone {boneName=dyn_belt_knife_r_02}
    componentAppliedLinearAccClamp=<5000 5000 5000>
    componentLinearAccScale=<1 1 1>
    componentLinearVelScale=<1 1 1>
    physicsBodyDefinitions [{
      sphereCollisionRadius=0
      boundBone {boneName=dyn_belt_knife_r_02}
      boxExtents=<12 22 22>
      constraintSetup {angularLimitsMax=<25 0 50>, angularLimitsMin=<-25 -10 -50>, angularTarget=<0 0 0>}
      localJointOffset=<1 0 0>
     }]
    relativeSpaceBone {boneName=root}
   }
   n165 = Anim Dynamics - Bone: dyn_back_strap_01(ComponentPose=@n164, Alpha=1, ExternalForce="0,0,0") {
    angularDampingOverride=0.9999
    angularSpringConstant=25
    bAngularSpring=true
    bGravityOverrideInSimSpace=true
    bOverrideAngularDamping=true
    simulationSpace=BoneRelative
    alphaBoolBlend {blendOption=Linear}
    boundBone {boneName=dyn_back_strap_01}
    componentAppliedLinearAccClamp=<3500 3500 3500>
    componentLinearAccScale=<1.5 1.5 1.5>
    componentLinearVelScale=<1.5 1.5 1.5>
    physicsBodyDefinitions [{
      sphereCollisionRadius=0
      boundBone {boneName=dyn_back_strap_01}
      boxExtents=<10 20 20>
      constraintSetup {angularLimitsMax=<25 0 25>, angularLimitsMin=<-25 -50 -25>, angularTarget=<0 0 0>}
      localJointOffset=<2 0 0>
     }]
    relativeSpaceBone {boneName=root}
   }
   n166 = Anim Dynamics - Bone: dyn_back_strap_02(ComponentPose=@n165, Alpha=1, ExternalForce="0,0,0") {
    angularDampingOverride=0.9999
    angularSpringConstant=25
    bAngularSpring=true
    bGravityOverrideInSimSpace=true
    bOverrideAngularDamping=true
    simulationSpace=BoneRelative
    alphaBoolBlend {blendOption=Linear}
    boundBone {boneName=dyn_back_strap_02}
    componentAppliedLinearAccClamp=<3500 3500 3500>
    componentLinearAccScale=<1.5 1.5 1.5>
    componentLinearVelScale=<1.5 1.5 1.5>
    physicsBodyDefinitions [{
      sphereCollisionRadius=0
      boundBone {boneName=dyn_back_strap_02}
      boxExtents=<10 20 20>
      constraintSetup {angularLimitsMax=<25 0 25>, angularLimitsMin=<-25 -50 -25>, angularTarget=<0 0 0>}
      localJointOffset=<2 0 0>
     }]
    relativeSpaceBone {boneName=root}
   }
   n167 = Anim Dynamics - Bone: dyn_back_strap_03(ComponentPose=@n166, Alpha=1, ExternalForce="0,0,0") {
    angularDampingOverride=0.9999
    angularSpringConstant=25
    bAngularSpring=true
    bGravityOverrideInSimSpace=true
    bOverrideAngularDamping=true
    simulationSpace=BoneRelative
    alphaBoolBlend {blendOption=Linear}
    boundBone {boneName=dyn_back_strap_03}
    componentAppliedLinearAccClamp=<3500 3500 3500>
    componentLinearAccScale=<1.5 1.5 1.5>
    componentLinearVelScale=<1.5 1.5 1.5>
    physicsBodyDefinitions [{
      sphereCollisionRadius=0
      boundBone {boneName=dyn_back_strap_03}
      boxExtents=<10 20 20>
      constraintSetup {angularLimitsMax=<25 0 25>, angularLimitsMin=<-25 -50 -25>, angularTarget=<0 0 0>}
      localJointOffset=<2 0 0>
     }]
    relativeSpaceBone {boneName=root}
   }
   n168 = "Constraint - dyn_back_strap_master (PPP)"(ComponentPose=@n167, Alpha=1, ConstraintWeights_0=0.0331, ConstraintWeights_1=0.157, ConstraintWeights_2=0.8099) {
    alphaBoolBlend {blendOption=Linear}
    boneToModify {boneName=dyn_back_strap_master}
    constraintSetup [{targetBone {boneName=spine_01}}, {targetBone {boneName=thigh_twist_01_r}}, {targetBone {boneName=pelvis}}]
    constraintWeights [1,1,1]
   }
   n169 = Get L_Arm_Back() -> (L_Arm_Back:double)
   n170 = "Constraint - dyn_poncho_armpit_back_l (P)"(ComponentPose=@n168, Alpha=@n169, ConstraintWeights_0=1) {
    alphaBoolBlend {blendOption=Linear}
    boneToModify {boneName=dyn_poncho_armpit_back_l}
    constraintSetup [{targetBone {boneName=upperarm_l}}]
    constraintWeights [1]
   }
   n171 = "Constraint - dyn_l_sleeve_back_01 (P)"(ComponentPose=@n170, Alpha=@n169, ConstraintWeights_0=1) {
    alphaBoolBlend {blendOption=Linear}
    boneToModify {boneName=dyn_l_sleeve_back_01}
    constraintSetup [{targetBone {boneName=upperarm_l}}]
    constraintWeights [1]
   }
   n172 = Get R_Arm_Back() -> (R_Arm_Back:double)
   n173 = "Constraint - dyn_poncho_armpit_back_r (P)"(ComponentPose=@n171, Alpha=@n172, ConstraintWeights_0=1) {
    alphaBoolBlend {blendOption=Linear}
    boneToModify {boneName=dyn_poncho_armpit_back_r}
    constraintSetup [{targetBone {boneName=upperarm_r}}]
    constraintWeights [1]
   }
   n174 = "Constraint - dyn_r_sleeve_back_01 (P)"(ComponentPose=@n173, Alpha=@n172, ConstraintWeights_0=1) {
    alphaBoolBlend {blendOption=Linear}
    boneToModify {boneName=dyn_r_sleeve_back_01}
    constraintSetup [{targetBone {boneName=upperarm_r}}]
    constraintWeights [1]
   }
   Save cached pose 'Pre_Melee'(Pose=Component To Local(ComponentPose=@n174)) {cacheName=Pre_Melee}
   SelectFloat(A=1, B=0, bPickA=Get bIsSprinting())
   SelectFloat(A=1, B=0, bPickA=Get bIsPlayingMeleeAnim())
   SelectFloat(A=1, B=0, bPickA=Get bIsCrouching())
   Get Melee_Float() -> (Melee_Float:double)
  ]}
 EventGraph {nodes [
   Event OnCharacterCustomizationLoaded()
   > Set Frontend_Kick_Out(Frontend_Kick_Out:double=1) -> (Output_Get:double)
   > Delay(Duration=0.1)
   > ResetDynamics(InTeleportType=TeleportPhysics)
   > ResetDynamics(InTeleportType=ResetPhysics)
   > Delay(Duration=0.1)
   > Set Frontend_Kick_Out(Frontend_Kick_Out=0)
   n1 = Event OnEndPlayingEmote() {is=title K2Node_Event}
   n2 = Delay(execute=[@n1.then, @n7.then], Duration=0.1)
   ResetDynamics(execute=[@n9.then, @n2.then], InTeleportType=TeleportPhysics)
   > ResetDynamics(InTeleportType=ResetPhysics)
   Event OnWeaponChanged()
   > n3 = Get Dynamic Material Instance(SkeletalMeshComponent=GetPartSkeletalMeshComponent())
   n4 = PickaxeItemDefinitionHasAnyMetaTag(Tags=Make Array("[0]"=(TagName="Cosmetics.MetaTags.WombatMale")), bRequiresEquipped=true)
   > n5 = Branch(Condition=@n4)
   > n6 = Branch(Condition=Get bIsSkyDiving())
   SetScalarParameterValue(execute=[@n8.then, @n23.else, @n6.then, @n5.else], self=@n3, ParameterName=Hide Element 03, Value=0)
   SetScalarParameterValue(execute=[@n23.then, @n6.else], self=@n3, ParameterName=Hide Element 03, Value=1)
   n7 = Event OnEndSkydiving()
   n8 = Event OnBeginSkydiving()
   Event BlueprintInitializeAnimation()
   > n9 = Delay(Duration=0.15)
   Event BlueprintUpdateAnimation()
   > n10 = Sequence()
   n11 = GetDeltaTransformFromRefPose(self=GetOwnerSkeletalMeshComponent(), BoneName=upperarm_r, BaseName=clavicle_r)
   n12 = MapRangeClamped(Value=BreakRotator(InRot=BreakTransform(InTransform=@n11).Rotation).Yaw, InRangeA=70, InRangeB=130)
   > Set Aim_Up(Aim_Up:double=@n12) -> (Output_Get:double)
   > Set Aimup_Off(Aimup_Off:double=float - float(A=1, B=@n12)) -> (Output_Get:double)
   n13 = BreakTransform(InTransform=GetDeltaTransformFromRefPose(self=GetOwnerSkeletalMeshComponent(), BoneName=upperarm_r, BaseName=clavicle_r))
   > Set Aim_Down(Aim_Down:double=MapRangeClamped(Value=BreakRotator(InRot=@n13.Rotation).Yaw, InRangeA=40, InRangeB=90)) -> (Output_Get:double)
   n14 = GetOwnerSkeletalMeshComponent()
   n15 = BreakRotator(InRot=BreakTransform(InTransform=GetDeltaTransformFromRefPose(self=@n14, BoneName=clavicle_l, BaseName=spine_05)).Rotation)
   n16 = MapRangeClamped(Value=@n15, InRangeA=0, InRangeB=10)
   > Set Rifle_Adjustment(Rifle_Adjustment:double=@n16) -> (Output_Get:double)
   > Set Rifle_Mid(Rifle_Mid:double=float - float(A=MapRangeClamped(Value=@n15, InRangeA=0, InRangeB=1), B=@n16) {is=title K2Node_CallFunction}) -> (Output_Get:double)
   n17 = GetOwnerSkeletalMeshComponent()
   n18 = BreakRotator(InRot=BreakTransform(InTransform=GetDeltaTransformFromRefPose(self=@n17, BoneName=upperarm_r, BaseName=clavicle_r)).Rotation)
   > Set Aim_Up_Variable(Aim_Up_Variable:double=float - float(A=MapRangeClamped(Value=@n18.Yaw, InRangeA=60, InRangeB=87), B=MapRangeClamped(Value=@n18.Yaw, InRangeA=87, InRangeB=130))) -> (Output_Get:double)
   > Set is_sprinting_float(is_sprinting_float:double=SelectFloat(A=1, B=0, bPickA=Get bIsSprinting())) -> (Output_Get:double)
   n19 = BreakTransform(InTransform=GetDeltaTransformFromRefPose(self=GetOwnerSkeletalMeshComponent(), BoneName=upperarm_l, BaseName=clavicle_l))
   > Set L_Arm_Back(L_Arm_Back:double=MapRangeClamped(Value=BreakRotator(InRot=@n19.Rotation).Yaw, InRangeA=0, InRangeB=-10)) -> (Output_Get:double)
   n20 = BreakTransform(InTransform=GetDeltaTransformFromRefPose(self=GetOwnerSkeletalMeshComponent(), BoneName=upperarm_r, BaseName=clavicle_r))
   > Set R_Arm_Back(R_Arm_Back:double=MapRangeClamped(Value=BreakRotator(InRot=@n20.Rotation).Yaw, InRangeA=0, InRangeB=-10)) -> (Output_Get:double)
   > Set inFrontend(inFrontend:double="To Float (Boolean)"(InBool=Get bIsInFrontEnd())) -> (Output_Get:double)
   > Set Crouch_Float(Crouch_Float:double=SelectFloat(A=1, B=0, bPickA=Get bIsCrouching())) -> (Output_Get:double)
   > Set Sprint_Float(Sprint_Float:double=0) -> (Output_Get:double)
   > Set Melee_Float(Melee_Float:double=SelectFloat(A=1, B=0, bPickA=Get bIsPlayingMeleeAnim())) -> (Output_Get:double)
   > Set JumpFall(JumpFall:bool=OR Boolean(A=Get bIsJumping(), B=Get bIsFalling())) -> (Output_Get:bool)
   > Set Skydive(Skydive:bool=OR Boolean(A=Get bIsSkyDiving(), B=Get bIsParachuteOpened())) -> (Output_Get:bool)
   > n21 = Cast To FortPlayerAnimInstance(Object=GetAnimInstance(self=GetOwnerSkeletalMeshComponent()))
   > Set melee_pickaxe(melee_pickaxe:double=0) -> (Output_Get:double)
   > Switch on EFortWeaponCoreAnimation!(execute, Selection:EFortWeaponCoreAnimation Enum=Get WeaponCoreAnim(self=@n21)) -> (
    Melee:Exec
    Pistol:Exec
    Shotgun:Exec
    PaperBlueprint:Exec
    Rifle:Exec
    MeleeOneHand:Exec
    MachinePistol:Exec
    RocketLauncher:Exec
    GrenadeLauncher:Exec
    GoingCommando:Exec
    AssaultRifle:Exec
    TacticalShotgun:Exec
    SniperRifle:Exec
    TrapPlacement:Exec
    ShoulderLauncher:Exec
    AbilityDecoTool:Exec
    Crossbow:Exec
    C4:Exec
    RemoteControl:Exec
    DualWield:Exec
    AR_BullPup:Exec
    AR_ForwardGrip:Exec
    MedPackPaddles:Exec
    SMG_P90:Exec
    AR_DrumGun:Exec
    Consumable_Small:Exec
    Consumable_Large:Exec
    Balloon:Exec
    MountedTurret:Exec
    CreativeTool:Exec
    ExplosiveBow:Exec
    AshtonIndigo:Exec
    AshtonChicago:Exec
    MeleeDualWield:Exec
    Unarmed:Exec
   )
   > Set melee_pickaxe(melee_pickaxe=1)
   n22 = Gravity Override State Function DEPRECATED(
    execute=@n10.then_1
    CharacterPartAnimInstance=Self-Reference()
    GameTimeInSeconds=GetGameTimeInSeconds()
    GravityOverrideParams=Get Gravity Override Params()
   )
   > Set GravityOverride(GravityOverride=@n22)
   n23 = Branch(Condition=@n4)
   n24 = GetDeltaTransformFromRefPose(self=GetOwnerSkeletalMeshComponent(), BoneName=clavicle_r, BaseName=spine_05)
   MapRangeClamped(Value=BreakRotator(InRot=BreakTransform(InTransform=@n24).Rotation).Pitch, InRangeA=17, InRangeB=32)
   SelectFloat(A=1, B=0, bPickA=Get bIsSprinting())
   ResetDynamics(execute=@n26.then, InTeleportType=ResetPhysics)
   n25 = Montage_IsPlaying(Montage=/BRCosmetics/Animation/Game/MainPlayer/Menu/BR/CustomSkins/_S15/Wombat/Wombat_Male_idle_M)
   > Do Once(Start Closed=NOT Boolean(A=@n25))
   > Do Once(Start Closed=@n25)
   > Delay(Duration=0.15)
   > n26 = ResetDynamics(InTeleportType=TeleportPhysics)
  ]}
 NewMacro_0 {nodes [
   Inputs!()
   Outputs()
  ]}
]
variables [
 Aim_Down {type=number}
 Aim_Up {type=number}
 Aim_Up_Variable {type=number}
 Aimup_Off {type=number}
 Crouch_Float {type=number}
 Frontend_Kick_Out {type=number}
 Gravity Override Params {
  type=GravityOverrideParamsStruct
  default {skydive_FloatingDown {gravityOverride=<0 -350 300>, windFrequency=<1 4 6>}, skydive_Parachuting {windFrequency=<1 4 6>}}
 }
 GravityOverride {type=Vector}
 inFrontend {type=number}
 is_sprinting_float {type=number}
 JumpFall {type=boolean}
 L_Arm_Back {type=number}
 Melee_Float {type=number}
 melee_pickaxe {type=number}
 R_Arm_Back {type=number}
 Rifle_Adjustment {type=number}
 Rifle_Mid {type=number}
 Skydive {type=boolean}
 Sprint_Float {type=number}
 Wombat_Idle_Pose {type=boolean}
]
